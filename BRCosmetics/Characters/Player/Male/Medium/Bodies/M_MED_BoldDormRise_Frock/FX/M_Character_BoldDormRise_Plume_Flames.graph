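# Unlit translucent material for flame trails on the BoldDormRise character's plume, used with Niagara ribbons and mesh particles.
# A panning smoke trail, layered flame noise and erosion shape opacity, softened by end, depth, view-angle and optional fresnel fades.
# Noise and length gradients blend three colors with a core highlight.
# Sine-wave world position offset, with optional SplineThicken, animates the shape; dynamic parameters drive wave intensity, U offset, erosion and width.
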Material /BRCosmetics/Characters/Player/Male/Medium/Bodies/M_MED_BoldDormRise_Frock/FX/M_Character_BoldDormRise_Plume_Flames
graph {
 properties {
  bAutomaticallySetUsageInEditor=false
  bUsedWithNiagaraMeshParticles=true
  bUsedWithNiagaraRibbons=true
  blendMode=BLEND_Translucent
  shadingModel=MSM_Unlit
  usedShadingModels=Unlit
 }
 nodes [
  Constant2Vector() {g=0.3, r=0.2}
  Constant2Vector() {g=0.7, r=1}
  Constant2Vector() {g=1.5, r=1}
  DynamicParameter() {parameterIndex=1, defaultValue=<rgba 0 1 1 1>, paramNames [Erosion,Param2,Param3,Param4]}
  n1 = ComponentMask(None=NamedRerouteUsage()) {r=true}
  n2 = Saturate(None=Multiply(A=Power(Base=@n1)) {constB=2})
  LinearInterpolate(Alpha=@n2)
  MaterialFunctionCall!(UV, UV Scale Pivot, UV Scale Start, UV Scale End, Duration, Time, Random UV Offset Enable, Random UV Offset Scale) {materialFunction=/BRCosmetics/Accessories/FORT_Backpacks/Backpack_M_MED_CrystalGlobe/FX/Materials/MF_ScaleCrossfadeUV}
  n3 = Multiply(A=ComponentMask(None=NamedRerouteUsage()) {r=true}) {constB=2}
  n4 = Multiply(
   A=LinearInterpolate(A=ScalarParameter() {defaultValue=0.5, group=Spline, parameterName=Spline_Base_Width}, Alpha=Saturate(None=@n3)) {constA=0.3}
   B=NamedRerouteUsage()
  ) {constA=1}
  n5 = Add(
   A=NamedRerouteUsage()
   B=AppendVector(A=DynamicParameter() {defaultValue=<rgba 50 0 0 0.2>, paramNames [WPOWave_Intensity, U_Offset, Erosion, SplineWIdth]}.U_Offset, B=Constant())
  )
  n6 = Time()
  n7 = Sine(None=Add(A=ComponentMask(None=@n5) {r=true}, B=Multiply(A=@n6) {constB=-0.36})) {period=1.2}
  n8 = Transform(None=VectorParameter() {parameterName=WPO Direction, defaultValue=<rgba 0.3 1 0 0>}.RGB) {transformSourceType=TRANSFORMSOURCE_Local}
  n9 = Multiply(
   A=Multiply(A=Multiply(A=@n7, B=@n8), B=@n2)
   B=DynamicParameter() {defaultValue=<rgba 50 0 0 0.2>, paramNames [WPOWave_Intensity, U_Offset, Erosion, SplineWIdth]}.WPOWave_Intensity
  ) {constB=100}
  n10 = NamedRerouteUsage()
  n11 = OneMinus(None=NamedRerouteUsage())
  n12 = NamedRerouteUsage()
  n13 = LinearInterpolate(
   A=ScalarParameter() {defaultValue=0.45, group=Trail, parameterName=TrailScaleLeading}
   B=ScalarParameter() {defaultValue=1.5, group=Trail, parameterName=TrailScaleEnding}
   Alpha=Power(Base=NamedRerouteUsage()) {constExponent=1}
  ) {constA=0.45, constB=1.5}
  n14 = ScalarParameter() {defaultValue=0.9, group=Trail, parameterName=ScaleFlame}
  n15 = Multiply(A=@n12, B=AppendVector(A=Multiply(A=@n13, B=@n14), B=Constant() {r=1}))
  n16 = AppendVector(
   A=Add(A=DynamicParameter() {defaultValue=<rgba 50 0 0 0.2>, paramNames [WPOWave_Intensity, U_Offset, Erosion, SplineWIdth]}.U_Offset) {constB=0.15}
   B=Constant()
  )
  n17 = ScalarParameter() {defaultValue=-1.1, group=Trail, parameterName=PanSpeedFlame}
  n18 = Constant()
  n19 = TextureSampleParameter2D(UVs=Panner(Coordinate=Add(A=@n15, B=@n16) {constB=0}, Speed=AppendVector(A=@n17, B=@n18)) {speedX=-0.3}) {parameterName=TrailTexture, samplerType=SAMPLERTYPE_Grayscale, texture=&T_Smoke_Trail_Soft}
  n20 = LinearInterpolate(A=Power(Base=@n19.R) {constExponent=1.25}, B=@n19.G, Alpha=ScalarParameter() {group=Tweaks, parameterName=UseGlow})
  n21 = Multiply(A=@n11, B=Add(A=@n11, B=@n20))
  n22 = ScalarParameter() {defaultValue=0.5, group=Trail, parameterName=FlameBaseGradient_Intensity}
  n23 = LinearInterpolate(
   A=Add(A=Multiply(A=@n21, B=@n22) {constB=0.5}, B=SmoothStep(Value=@n20))
   B=@n20
   Alpha=ScalarParameter() {group=Tweaks, parameterName=UseGlow}
  )
  n24 = Multiply(A=NamedRerouteUsage(), B=Multiply(A=NamedRerouteUsage(), B=@n23))
  n25 = ComponentMask(None=NamedRerouteUsage()) {g=true}
  n26 = Saturate(None=Sine(None=Multiply(A=@n25) {constB=0.5}))
  n27 = Multiply(A=Power(Base=Multiply(A=@n24, B=@n26)) {constExponent=0.5}) {constB=5}
  n28 = Constant3Vector() {constant=<rgba 0 0 1 0>}
  n29 = Saturate(None=@n7)
  n30 = Multiply(A=Saturate(None=Multiply(A=@n10) {constB=10}), B=Add(A=Multiply(A=@n27, B=@n28), B=Multiply(A=@n28, B=@n29)))
  Multiply(
   A=StaticSwitchParameter(True=Add(A=@n4, B=@n9) {constB=0}, False=Add(A=@n30, B=@n9)) {group=Tweaks, parameterName=UseSplineThicken}
   B=ScalarParameter() {defaultValue=1, parameterName=WPO_Intenisty}
  )
  n31 = NamedRerouteUsage()
  n32 = LinearInterpolate(Alpha=DynamicParameter() {defaultValue=<rgba 50 0 0 0.2>, paramNames [WPOWave_Intensity, U_Offset, Erosion, SplineWIdth]}.Erosion) {constA=-1}
  n33 = NamedRerouteUsage()
  n34 = Constant() {r=0.5}
  n35 = ComponentMask(None=@n33) {r=true}
  n36 = Multiply(
   A=LinearInterpolate(A=SphereMask(A=@n33, B=@n34) {attenuationRadius=0.5, hardnessPercent=0}, Alpha=@n35)
   B=ScalarParameter() {defaultValue=10, parameterName=FadeLeadingEdge_MaskIntensity}
  ) {constB=10}
  n37 = NamedRerouteUsage()
  n38 = NamedRerouteUsage()
  n39 = Multiply(A=@n36, B=LinearInterpolate(A=Multiply(A=@n37, B=@n38), Alpha=@n35))
  n40 = Power(Base=OneMinus(None=@n10), Exp=ScalarParameter() {defaultValue=1, group=Alpha, parameterName=TrailingEndFadePower})
  n41 = NamedRerouteUsage()
  n42 = SmoothStep(Value=@n41)
  n43 = OneMinus(None=NamedRerouteUsage())
  n44 = Multiply(
   A=Power(Base=@n43, Exp=ScalarParameter() {defaultValue=4, group=Core, parameterName=CoreGradientPower}) {constExponent=4}
   B=NamedRerouteUsage()
  )
  n45 = Saturate(None=Power(Base=Saturate(None=@n44), Exp=ScalarParameter() {defaultValue=50, group=Core, parameterName=CoreSizePower}) {constExponent=1})
  n46 = LinearInterpolate(
   A=LinearInterpolate(A=Saturate(None=@n42), B=@n41, Alpha=ScalarParameter() {group=Tweaks, parameterName=UseGlow})
   B=@n45
   Alpha=ScalarParameter() {group=Tweaks, parameterName=Only_Core}
  ) {constB=0}
  n47 = Multiply(
   A=LinearInterpolate(B=Saturate(None=@n39), Alpha=ScalarParameter() {defaultValue=1, parameterName=FadeLeadingEnd}) {constA=1}
   B=LinearInterpolate(A=@n46, B=Multiply(A=@n46, B=@n40), Alpha=ScalarParameter() {parameterName=FadeTrailingEnd})
  )
  n48 = DepthFade(Opacity=@n47, FadeDistance=ScalarParameter() {defaultValue=6, group=Alpha, parameterName=FadeDistance}) {fadeDistanceDefault=20}
  n49 = QualitySwitch(Default=@n48, Low=@n47, High=@n48, Medium=@n47, Epic=@n48)
  n50 = StaticSwitchParameter(True=Fresnel(Normal=NamedRerouteUsage()) {baseReflectFraction=0, exponent=1}, False=Fresnel() {baseReflectFraction=0, exponent=1}) {group=Tweaks, parameterName=UseSplineThicken}
  n51 = Multiply(
   A=Power(Base=OneMinus(None=@n50), Exp=ScalarParameter() {defaultValue=1, group=Alpha, parameterName=Fresnel_Power})
   B=ScalarParameter() {defaultValue=1, group=Alpha, parameterName=Fresnel_Intensity}
  )
  n52 = Multiply(
   A=LinearInterpolate(B=Saturate(None=@n51), Alpha=ScalarParameter() {group=Alpha, parameterName=UseFresnel}) {constA=1}
   B=MaterialFunctionCall("FadeSensitivity (optional)"=ScalarParameter() {group=Alpha, parameterName=FadeBasedOnViewAngle}) {materialFunction=&FadeBasedOnViewAngle}.Result
  )
  Multiply(A=Saturate(None=Subtract(A=@n31, B=@n32)), B=Multiply(A=ParticleColor().A, B=Multiply(A=@n49, B=@n52)))
  n53 = NamedRerouteUsage()
  n54 = OneMinus(None=ComponentMask(None=@n53) {r=true})
  n55 = ScalarParameter() {defaultValue=5, group=Color, parameterName=ColorA_BaseGradientBias_Power}
  n56 = AppendVector(A=ScalarParameter() {defaultValue=1, parameterName=Noise_ScaleU}, B=ScalarParameter() {defaultValue=1.5, parameterName=Noise_ScaleV})
  n57 = ScalarParameter() {defaultValue=-0.77, parameterName=Noise_PanSpeedU}
  n58 = ScalarParameter() {defaultValue=0.1, parameterName=Noise_PanSpeedV}
  n59 = TextureSampleParameter2D(UVs=Panner(Coordinate=Multiply(A=@n53, B=@n56), Speed=AppendVector(A=@n57, B=@n58)) {speedX=-0.77, speedY=0.1}) {parameterName=Noise2, texture=/BRCosmetics/Athena/Effects/Particles/Skydiving/NS_Contrail_GrimHound/T_FX_Contrail_GrimHound_Fire}
  n60 = NamedRerouteUsage()
  n61 = ScalarParameter() {defaultValue=0.2, parameterName=Flames1_ScaleU}
  n62 = ScalarParameter() {defaultValue=0.3, parameterName=Flames1_ScaleV}
  n63 = LinearInterpolate(A=AppendVector(A=@n61, B=@n62), B=AppendVector(A=@n62, B=@n61), Alpha=ScalarParameter() {parameterName=SwizzleUVs})
  n64 = ScalarParameter() {defaultValue=-0.3, parameterName=Flames1_PanSpeedU}
  n65 = ScalarParameter() {defaultValue=0.05, parameterName=Flames1_PanSPeedV}
  n66 = TextureSampleParameter2D(UVs=Panner(Coordinate=Multiply(A=@n60, B=@n63), Speed=AppendVector(A=@n64, B=@n65)) {speedX=-0.3}) {parameterName=Flame_Texture, samplerType=SAMPLERTYPE_LinearColor, texture=&T_NoisePack_LinearColor}
  n67 = Add(
   A=Multiply(A=Power(Base=@n54, Exp=@n55) {constExponent=5}, B=ScalarParameter() {defaultValue=1, group=Color, parameterName=ColorA_BaseGradientBias})
   B=Multiply(A=@n59.R, B=LinearInterpolate(Alpha=@n66.R))
  )
  n68 = Subtract(A=@n67, B=ScalarParameter() {defaultValue=0.4, group=Color, parameterName=ColorASubtractBias}) {constB=0.4}
  Multiply(A=@n68) {constB=100}
  n69 = Subtract(A=@n67, B=ScalarParameter() {defaultValue=0.1, group=Color, parameterName=ColorBSubtractBias}) {constB=0.1}
  Multiply(A=@n69) {constB=100}
  n70 = ScalarParameter() {defaultValue=2, group=Alpha, parameterName=TrailIntensity}
  n71 = NamedRerouteUsage()
  n72 = SmoothStep(Value=@n71) {constMax=0.8, constMin=0.1}
  n73 = Subtract(
   A=Multiply(A=NamedRerouteUsage(), B=Multiply(A=@n70, B=@n23) {constB=5})
   B=LinearInterpolate(B=Multiply(A=@n72), Alpha=ScalarParameter() {defaultValue=1, group=Tweaks, parameterName=UseErosion})
  )
  NamedRerouteDeclaration(None=@n73) {name=MainTrail, nodeColor=<rgba 0 0.4375 1 1>}
  n74 = Constant() {r=100}
  n75 = DynamicParameter() {defaultValue=<rgba 50 0 0 0.2>, paramNames [WPOWave_Intensity, U_Offset, Erosion, SplineWIdth]}
  n76 = Constant() {r=100}
  n77 = MaterialFunctionCall(
   WidthBase=Multiply(A=Multiply(A=@n74, B=@n75.SplineWIdth), B=ScalarParameter() {defaultValue=1, group=Spline, parameterName=SplineWidthBase_Multiplier})
   WidthTip=Multiply(A=Multiply(A=@n76, B=@n75.SplineWIdth), B=ScalarParameter() {defaultValue=1, group=Spline, parameterName=SplineWidthTip_Multiplier})
  ) {materialFunction=&SplineThicken}
  NamedRerouteDeclaration(None=@n77.Normal) {name=Spline Normals, nodeColor=<rgba 0.8949 0.2818 1 1>}
  NamedRerouteDeclaration(None=@n77.WorldPositionOffset) {name=Spline WPO, nodeColor=<rgba 0 1 0.3906 1>}
  NamedRerouteDeclaration(None=@n59.R) {name=FlameNoise, nodeColor=<rgba 0 1 0.3672 1>}
  n78 = Multiply(A=@n11, B=ScalarParameter() {defaultValue=2, parameterName=LengthGradient_Intensity}) {constB=10}
  n79 = Sine(None=Multiply(A=ComponentMask(None=@n12) {g=true}) {constB=0.5})
  NamedRerouteDeclaration(None=Multiply(A=Saturate(None=@n78), B=Saturate(None=@n79))) {name=EdgeFades, nodeColor=<rgba 1 0.3984 0 1>}
  n80 = TextureCoordinate()
  n81 = Multiply(A=@n80, B=Constant2Vector() {g=1, r=1})
  n82 = MaterialFunctionCall(XYZ=@n80) {materialFunction=&Swizzle}
  NamedRerouteDeclaration(None=LinearInterpolate(A=@n81, B=@n82.YX, Alpha=ScalarParameter() {parameterName=SwizzleUVs})) {name=UVs, nodeColor=<rgba 1 0 0.1875 1>}
  n83 = NamedRerouteUsage()
  n84 = ScalarParameter() {defaultValue=1, parameterName=Flames2_ScaleU}
  n85 = ScalarParameter() {defaultValue=0.7, parameterName=Flames2_ScaleV}
  n86 = LinearInterpolate(A=AppendVector(A=@n84, B=@n85), B=AppendVector(A=@n85, B=@n84), Alpha=ScalarParameter() {parameterName=SwizzleUVs})
  n87 = ScalarParameter() {defaultValue=-0.55, parameterName=Flames2_PanSpeedU}
  n88 = ScalarParameter() {defaultValue=-0.066, parameterName=Flames2_PanSPeedV}
  n89 = TextureSampleParameter2D(UVs=Panner(Coordinate=Multiply(A=@n83, B=@n86), Speed=AppendVector(A=@n87, B=@n88)) {speedX=-0.55, speedY=-0.1}) {parameterName=Flame_Texture, samplerType=SAMPLERTYPE_LinearColor, texture=&T_NoisePack_LinearColor}
  n90 = NamedRerouteUsage()
  n91 = Multiply(
   A=Power(Base=ComponentMask(None=@n90) {r=true}, Exp=ScalarParameter() {defaultValue=1, parameterName=ErosionScaleCutoff_Power}) {constExponent=1}
   B=ScalarParameter() {defaultValue=1, parameterName=ErosionScaleCutoff_Intensity}
  )
  NamedRerouteDeclaration(None=LinearInterpolate(A=@n66.R, B=@n89.R, Alpha=Saturate(None=@n91))) {name=Flames, nodeColor=<rgba 1 0.225 0.0758 1>}
  NamedRerouteDeclaration(None=ComponentMask(None=@n12) {r=true}) {name=FadeGradient, nodeColor=<rgba 1 0 0.1172 1>}
  NamedRerouteUsage()
  OneMinus(None=@n80)
  n92 = Multiply(
   A=ComponentMask(None=NamedRerouteUsage()) {r=true}
   B=ScalarParameter() {defaultValue=5, group=Spline, parameterName=SplineTicken Edge Contrast}
  ) {constB=5}
  Reroute(None=LinearInterpolate(Alpha=Saturate(None=@n92)) {constA=0.5})
  n93 = ParticleColor()
  n94 = VectorParameter() {group=Color, parameterName=ColorC, defaultValue=<rgba 0.7 0.2013 0.035 1>}
  n95 = VectorParameter() {group=Color, parameterName=ColorB, defaultValue=<rgba 1 0.335 0.05 1>}
  n96 = Saturate(None=SmoothStep(Value=@n69) {constMax=0.66, constMin=0.05})
  n97 = SmoothStep(Value=@n68) {constMax=0.5}
  n98 = Multiply(
   A=@n93.RGB
   B=LinearInterpolate(
    A=LinearInterpolate(A=@n94.RGB, B=@n95.RGB, Alpha=@n96)
    B=VectorParameter() {group=Color, parameterName=ColorA, defaultValue=<rgba 1.2 0.63 0.06 1>}.RGB
    Alpha=Saturate(None=@n97)
   ) {constAlpha=0}
  )
  n99 = VectorParameter() {group=Core, parameterName=ColorCore, defaultValue=<rgba 1 1 1 1>}
  n100 = Power(Base=Power(Base=Multiply(A=@n43)) {constExponent=7})
  n101 = Saturate(None=Max(A=Multiply(A=@n100) {constB=0.3}, B=@n45))
  n102 = LinearInterpolate(
   A=@n98
   B=LinearInterpolate(
    A=LinearInterpolate(A=@n98, B=@n99.RGB, Alpha=@n101)
    B=Multiply(A=@n99.RGB, B=@n93.RGB)
    Alpha=ScalarParameter() {group=Tweaks, parameterName=Only_Core}
   ) {constB=0}
   Alpha=ScalarParameter() {defaultValue=1, group=Core, parameterName=CoreIntensity}
  )
  n103 = ParticleColor()
  n104 = ScalarParameter() {group=Tweaks, parameterName=UseGlow}
  Reroute(None=MaterialFunctionCall(Emissive=LinearInterpolate(A=@n102, B=@n103.RGB, Alpha=@n104)) {materialFunction=&MF_FX_EmissiveAdjustments}.Result)
  Saturate(None=@n73)
  Saturate(None=@n21)
  Saturate(None=Subtract(A=@n36, B=@n37))
  ScalarParameter() {defaultValue=1, parameterName=Width}
  SmoothStep(Value=@n41) {constMax=0.2, constMin=0.1}
  SmoothStep(Value=@n71) {constMax=0.8, constMin=0.3}
  StaticBoolParameter() {parameterName=ExpandUorV}
  StaticSwitchParameter(True=@n82.YX, False=@n81) {parameterName=SwizzleUVs}
  TextureCoordinate()
  TextureCoordinate()
  TextureCoordinate() {uTiling=2.2}
  VectorParameter() {group=Color, parameterName=OLD_ColorC, defaultValue=<rgba 0.4 0.3177 0.02 1>}
  VectorParameter() {group=Color, parameterName=OLD_ColorB, defaultValue=<rgba 0.5 0.4428 0.01 1>}
  VectorParameter() {group=Color, parameterName=OLD_ColorA, defaultValue=<rgba 1.2 1.0628 0.024 1>}
  VertexNormalWS()
 ]
}
