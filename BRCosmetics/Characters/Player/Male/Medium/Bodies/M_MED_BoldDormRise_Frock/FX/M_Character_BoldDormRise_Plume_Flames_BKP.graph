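# Translucent unlit flame-plume material for the BoldDormRise character, flagged for Niagara mesh particles and ribbons; the name suggests a backup.
# Panning noise and trail textures build an eroded flame, tinted by three colors and particle color, with a bright core and glow option.
# Opacity combines end fades, Fresnel and quality-switched depth fade.
# World position offset adds a sine wave, with optional SplineThicken width, driven by dynamic parameters.
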Material /BRCosmetics/Characters/Player/Male/Medium/Bodies/M_MED_BoldDormRise_Frock/FX/M_Character_BoldDormRise_Plume_Flames_BKP
graph {
 properties {
  bAutomaticallySetUsageInEditor=false
  bUsedWithNiagaraMeshParticles=true
  bUsedWithNiagaraRibbons=true
  blendMode=BLEND_Translucent
  shadingModel=MSM_Unlit
  usedShadingModels=Unlit
 }
 nodes [
  n1 = OneMinus(None=NamedRerouteUsage())
  n2 = Multiply(
   A=Power(Base=@n1, Exp=ScalarParameter() {defaultValue=4, group=Core, parameterName=CoreGradientPower}) {constExponent=4}
   B=NamedRerouteUsage()
  )
  Abs(None=Multiply(A=Power(Base=@n2) {constExponent=10000}))
  Constant2Vector() {g=0.3, r=0.2}
  Constant2Vector() {g=0.7, r=1}
  DynamicParameter() {parameterIndex=1, defaultValue=<rgba 0 1 1 1>, paramNames [Erosion,Param2,Param3,Param4]}
  n3 = ComponentMask(None=NamedRerouteUsage()) {r=true}
  n4 = Saturate(None=Multiply(A=Power(Base=@n3)) {constB=2})
  LinearInterpolate(Alpha=@n4)
  MaterialFunctionCall!(UV, UV Scale Pivot, UV Scale Start, UV Scale End, Duration, Time, Random UV Offset Enable, Random UV Offset Scale) {materialFunction=/BRCosmetics/Accessories/FORT_Backpacks/Backpack_M_MED_CrystalGlobe/FX/Materials/MF_ScaleCrossfadeUV}
  n5 = NamedRerouteUsage()
  n6 = LinearInterpolate(Alpha=DynamicParameter() {defaultValue=<rgba 50 0 0 0.2>, paramNames [WPOWave_Intensity, U_Offset, Erosion, SplineWIdth]}.Erosion) {constA=-1}
  n7 = NamedRerouteUsage()
  n8 = SphereMask(A=@n7, B=Constant() {r=0.5}) {attenuationRadius=0.5, hardnessPercent=0}
  n9 = ComponentMask(None=@n7) {r=true}
  n10 = Subtract(A=Multiply(A=LinearInterpolate(A=@n8, Alpha=@n9)) {constB=10}, B=NamedRerouteUsage())
  n11 = NamedRerouteUsage()
  n12 = Power(Base=OneMinus(None=@n11), Exp=ScalarParameter() {defaultValue=1, group=Alpha, parameterName=TrailingEndFadePower})
  n13 = NamedRerouteUsage()
  n14 = Saturate(None=Power(Base=Saturate(None=@n2), Exp=ScalarParameter() {defaultValue=50, group=Core, parameterName=CoreSizePower}) {constExponent=1})
  n15 = LinearInterpolate(
   A=LinearInterpolate(A=SmoothStep(Value=@n13) {constMax=0.2, constMin=0.1}, B=@n13, Alpha=ScalarParameter() {group=Tweaks, parameterName=UseGlow})
   B=@n14
   Alpha=ScalarParameter() {group=Tweaks, parameterName=Only_Core}
  ) {constB=0}
  n16 = Multiply(
   A=LinearInterpolate(B=Saturate(None=@n10), Alpha=ScalarParameter() {parameterName=FadeLeadingEnd}) {constA=1}
   B=LinearInterpolate(A=@n15, B=Multiply(A=@n15, B=@n12), Alpha=ScalarParameter() {parameterName=FadeTrailingEnd})
  )
  n17 = DepthFade(Opacity=@n16, FadeDistance=ScalarParameter() {defaultValue=6, group=Alpha, parameterName=FadeDistance}) {fadeDistanceDefault=20}
  n18 = QualitySwitch(Default=@n17, Low=@n16, High=@n17, Medium=@n16, Epic=@n17)
  n19 = Fresnel(Normal=NamedRerouteUsage()) {baseReflectFraction=0, exponent=1}
  n20 = Fresnel() {baseReflectFraction=0, exponent=1}
  n21 = Power(
   Base=OneMinus(None=StaticSwitchParameter(True=@n19, False=@n20) {group=Tweaks, parameterName=UseSplineThicken})
   Exp=ScalarParameter() {defaultValue=1, group=Alpha, parameterName=Fresnel_Power}
  )
  n22 = ScalarParameter() {defaultValue=1, group=Alpha, parameterName=Fresnel_Intensity}
  n23 = LinearInterpolate(B=Saturate(None=Multiply(A=@n21, B=@n22)), Alpha=ScalarParameter() {group=Alpha, parameterName=UseFresnel}) {constA=1}
  Multiply(A=Saturate(None=Subtract(A=@n5, B=@n6)), B=Multiply(A=ParticleColor().A, B=Multiply(A=@n18, B=@n23)))
  n24 = NamedRerouteUsage()
  n25 = OneMinus(None=ComponentMask(None=@n24) {r=true})
  n26 = Constant2Vector() {g=1.5, r=1}
  n27 = TextureSampleParameter2D(UVs=Panner(Coordinate=Multiply(A=@n24, B=@n26)) {speedX=-0.77, speedY=0.1}) {parameterName=Noise2, texture=/BRCosmetics/Athena/Effects/Particles/Skydiving/NS_Contrail_GrimHound/T_FX_Contrail_GrimHound_Fire}
  n28 = NamedRerouteUsage()
  n29 = ScalarParameter() {defaultValue=0.2, parameterName=Flames1_ScaleU}
  n30 = ScalarParameter() {defaultValue=0.3, parameterName=Flames1_ScaleV}
  n31 = LinearInterpolate(A=AppendVector(A=@n29, B=@n30), B=AppendVector(A=@n30, B=@n29), Alpha=ScalarParameter() {parameterName=SwizzleUVs})
  n32 = ScalarParameter() {defaultValue=-0.3, parameterName=Flames1_PanSpeedU}
  n33 = ScalarParameter() {defaultValue=0.05, parameterName=Flames1_PanSPeedV}
  n34 = TextureSampleParameter2D(UVs=Panner(Coordinate=Multiply(A=@n28, B=@n31), Speed=AppendVector(A=@n32, B=@n33)) {speedX=-0.3}) {
   parameterName=Flame_Texture
   samplerType=SAMPLERTYPE_LinearColor
   texture=&/T_BoldDormRise_Crystal_Noise
  }
  n35 = Add(A=Multiply(A=Power(Base=@n25) {constExponent=5}), B=Multiply(A=@n27.R, B=LinearInterpolate(Alpha=@n34.R)))
  n36 = Subtract(A=@n35) {constB=0.17}
  Multiply(A=@n36) {constB=100}
  n37 = Multiply(A=ComponentMask(None=NamedRerouteUsage()) {r=true}) {constB=4}
  n38 = Multiply(
   A=LinearInterpolate(A=ScalarParameter() {defaultValue=0.3, group=Spline, parameterName=Spline_Base_Width}, Alpha=Saturate(None=@n37)) {constA=0.3}
   B=NamedRerouteUsage()
  )
  n39 = Add(
   A=NamedRerouteUsage()
   B=AppendVector(A=DynamicParameter() {defaultValue=<rgba 50 0 0 0.2>, paramNames [WPOWave_Intensity, U_Offset, Erosion, SplineWIdth]}.U_Offset, B=Constant())
  )
  n40 = Time()
  n41 = Sine(None=Add(A=ComponentMask(None=@n39) {r=true}, B=Multiply(A=@n40) {constB=-0.36})) {period=1.2}
  n42 = Transform(None=VectorParameter() {parameterName=WPO Direction, defaultValue=<rgba 0.3 1 0 0>}.RGB) {transformSourceType=TRANSFORMSOURCE_Local}
  n43 = Multiply(
   A=Multiply(A=Multiply(A=@n41, B=@n42), B=@n4)
   B=DynamicParameter() {defaultValue=<rgba 50 0 0 0.2>, paramNames [WPOWave_Intensity, U_Offset, Erosion, SplineWIdth]}.WPOWave_Intensity
  ) {constB=100}
  n44 = OneMinus(None=NamedRerouteUsage())
  n45 = NamedRerouteUsage()
  n46 = ComponentMask(None=@n45) {g=true}
  n47 = Saturate(None=Sine(None=Multiply(A=@n46) {constB=0.5}))
  n48 = LinearInterpolate(Alpha=Power(Base=NamedRerouteUsage()) {constExponent=1}) {constA=0.1, constB=0.7}
  n49 = ScalarParameter() {defaultValue=0.9, group=Trail, parameterName=ScaleFlame}
  n50 = Multiply(A=@n45, B=AppendVector(A=Multiply(A=@n48, B=@n49), B=Constant() {r=1}))
  n51 = AppendVector(
   A=Add(A=DynamicParameter() {defaultValue=<rgba 50 0 0 0.2>, paramNames [WPOWave_Intensity, U_Offset, Erosion, SplineWIdth]}.U_Offset) {constB=0.15}
   B=Constant()
  )
  n52 = ScalarParameter() {defaultValue=-0.85, group=Trail, parameterName=PanSpeedFlame}
  n53 = Constant()
  n54 = TextureSampleParameter2D(UVs=Panner(Coordinate=Add(A=@n50, B=@n51) {constB=0}, Speed=AppendVector(A=@n52, B=@n53)) {speedX=-0.3}) {
   parameterName=TrailTexture
   texture=&/T_BoldDormRise_Trail_Combined
  }
  n55 = LinearInterpolate(A=@n54.R, B=@n54.R, Alpha=ScalarParameter() {group=Tweaks, parameterName=UseGlow})
  n56 = Multiply(A=@n44, B=Add(A=@n44, B=@n55))
  n57 = Add(
   A=Multiply(A=Saturate(None=@n56), B=ScalarParameter() {defaultValue=0.5, group=Trail, parameterName=FlameBaseGradient_Intensity}) {constB=0.5}
   B=SmoothStep(Value=@n55)
  )
  n58 = ScalarParameter() {group=Tweaks, parameterName=UseGlow}
  n59 = StaticSwitchParameter(True=Multiply(A=Multiply(A=@n44, B=@n47), B=LinearInterpolate(A=@n57, B=@n55, Alpha=@n58)), False=@n44) {defaultValue=true, parameterName=UseTrailTexture}
  n60 = Multiply(A=NamedRerouteUsage(), B=@n59)
  n61 = Multiply(A=Power(Base=Multiply(A=@n60, B=@n47)) {constExponent=0.5}) {constB=5}
  n62 = Constant3Vector() {constant=<rgba 0 0 1 0>}
  n63 = Saturate(None=@n41)
  n64 = Multiply(A=Saturate(None=Multiply(A=@n11) {constB=10}), B=Add(A=Multiply(A=@n61, B=@n62), B=Multiply(A=@n62, B=@n63)))
  Multiply(
   A=StaticSwitchParameter(True=Add(A=@n38, B=@n43) {constB=0}, False=Add(A=@n64, B=@n43)) {group=Tweaks, parameterName=UseSplineThicken}
   B=ScalarParameter() {defaultValue=1, parameterName=WPO_Intenisty}
  )
  n65 = TextureCoordinate()
  n66 = Multiply(A=@n65, B=Constant2Vector() {g=1, r=1})
  n67 = MaterialFunctionCall(XYZ=@n65) {materialFunction=&Swizzle}
  NamedRerouteDeclaration(None=LinearInterpolate(A=@n66, B=@n67.YX, Alpha=ScalarParameter() {parameterName=SwizzleUVs})) {name=UVs, nodeColor=<rgba 1 0 0.1875 1>}
  n68 = NamedRerouteUsage()
  n69 = ScalarParameter() {defaultValue=1, parameterName=Flames2_ScaleU}
  n70 = ScalarParameter() {defaultValue=0.7, parameterName=Flames2_ScaleV}
  n71 = LinearInterpolate(A=AppendVector(A=@n69, B=@n70), B=AppendVector(A=@n70, B=@n69), Alpha=ScalarParameter() {parameterName=SwizzleUVs})
  n72 = ScalarParameter() {defaultValue=-0.55, parameterName=Flames2_PanSpeedU}
  n73 = ScalarParameter() {defaultValue=-0.1, parameterName=Flames2_PanSPeedV}
  n74 = TextureSampleParameter2D(UVs=Panner(Coordinate=Multiply(A=@n68, B=@n71), Speed=AppendVector(A=@n72, B=@n73)) {speedX=-0.55, speedY=-0.1}) {
   parameterName=Flame_Texture
   samplerType=SAMPLERTYPE_LinearColor
   texture=&T_BoldDormRise_Crystal_Noise
  }
  n75 = NamedRerouteUsage()
  n76 = Multiply(
   A=Power(Base=ComponentMask(None=@n75) {r=true}, Exp=ScalarParameter() {defaultValue=1, parameterName=ErosionScaleCutoff_Power}) {constExponent=1}
   B=ScalarParameter() {defaultValue=1, parameterName=ErosionScaleCutoff_Intensity}
  )
  NamedRerouteDeclaration(None=LinearInterpolate(A=@n34.R, B=@n74.R, Alpha=Saturate(None=@n76))) {name=Flames, nodeColor=<rgba 1 0.225 0.0758 1>}
  NamedRerouteDeclaration(None=ComponentMask(None=@n45) {r=true}) {name=FadeGradient, nodeColor=<rgba 1 0 0.1172 1>}
  n77 = SmoothStep(Value=NamedRerouteUsage()) {constMax=0.8, constMin=0.3}
  n78 = Subtract(
   A=Multiply(A=ScalarParameter() {defaultValue=2, group=Alpha, parameterName=TrailIntensity}, B=@n59) {constB=5}
   B=LinearInterpolate(B=Multiply(A=@n77), Alpha=ScalarParameter() {defaultValue=1, group=Tweaks, parameterName=UseErosion})
  )
  NamedRerouteDeclaration(None=@n78) {name=MainTrail, nodeColor=<rgba 0 0.4375 1 1>}
  n79 = Constant() {r=100}
  n80 = DynamicParameter() {defaultValue=<rgba 50 0 0 0.2>, paramNames [WPOWave_Intensity, U_Offset, Erosion, SplineWIdth]}
  n81 = Constant() {r=100}
  n82 = MaterialFunctionCall(
   WidthBase=Multiply(A=Multiply(A=@n79, B=@n80.SplineWIdth), B=ScalarParameter() {defaultValue=1, group=Spline, parameterName=SplineWidthBase_Multiplier})
   WidthTip=Multiply(A=Multiply(A=@n81, B=@n80.SplineWIdth), B=ScalarParameter() {defaultValue=1, group=Spline, parameterName=SplineWidthTip_Multiplier})
  ) {materialFunction=&SplineThicken}
  NamedRerouteDeclaration(None=@n82.Normal) {name=Spline Normals, nodeColor=<rgba 0.8949 0.2818 1 1>}
  NamedRerouteDeclaration(None=@n82.WorldPositionOffset) {name=Spline WPO, nodeColor=<rgba 0 1 0.3906 1>}
  NamedRerouteDeclaration(None=@n27.R) {name=FlameNoise, nodeColor=<rgba 0 1 0.3672 1>}
  NamedRerouteUsage()
  OneMinus(None=@n65)
  Power(Base=@n54.G, Exp=ScalarParameter() {defaultValue=1, parameterName=PowerGlow})
  Power(Base=@n34.R)
  n83 = Multiply(
   A=ComponentMask(None=NamedRerouteUsage()) {r=true}
   B=ScalarParameter() {defaultValue=5, group=Spline, parameterName=SplineTicken Edge Contrast}
  ) {constB=5}
  Reroute(None=LinearInterpolate(Alpha=Saturate(None=@n83)) {constA=0.5})
  n84 = VectorParameter() {group=Color, parameterName=ColorC, defaultValue=<rgba 0.7 0.2013 0.035 1>}
  n85 = VectorParameter() {group=Color, parameterName=ColorB, defaultValue=<rgba 1 0.335 0.05 1>}
  n86 = Saturate(None=SmoothStep(Value=@n36) {constMax=0.07})
  n87 = Multiply(A=Subtract(A=@n35) {constB=0.35}) {constB=100}
  n88 = Multiply(
   A=ParticleColor().RGB
   B=LinearInterpolate(
    A=LinearInterpolate(A=@n84.RGB, B=@n85.RGB, Alpha=@n86)
    B=VectorParameter() {group=Color, parameterName=ColorA, defaultValue=<rgba 1.2 0.63 0.06 1>}.RGB
    Alpha=Saturate(None=@n87)
   )
  )
  n89 = VectorParameter() {group=Core, parameterName=ColorCore, defaultValue=<rgba 1 1 1 1>}
  n90 = Power(Base=Power(Base=Multiply(A=@n1)) {constExponent=7})
  n91 = Saturate(None=Max(A=Multiply(A=@n90) {constB=0.3}, B=@n14))
  n92 = ParticleColor()
  n93 = LinearInterpolate(
   A=@n88
   B=LinearInterpolate(
    A=LinearInterpolate(A=@n88, B=@n89.RGB, Alpha=@n91)
    B=Multiply(A=@n89.RGB, B=@n92.RGB)
    Alpha=ScalarParameter() {group=Tweaks, parameterName=Only_Core}
   ) {constB=0}
   Alpha=ScalarParameter() {defaultValue=1, group=Core, parameterName=CoreIntensity}
  )
  n94 = ScalarParameter() {group=Tweaks, parameterName=UseGlow}
  Reroute(None=MaterialFunctionCall(Emissive=LinearInterpolate(A=@n93, B=@n92.RGB, Alpha=@n94)) {materialFunction=&MF_FX_EmissiveAdjustments}.Result)
  Saturate(None=@n78)
  ScalarParameter() {defaultValue=1, parameterName=Width}
  SmoothStep(Value=@n2) {constMax=0.4, constMin=0.4}
  StaticBoolParameter() {parameterName=ExpandUorV}
  StaticSwitchParameter(True=@n67.YX, False=@n66) {parameterName=SwizzleUVs}
  TextureCoordinate()
  TextureCoordinate()
  TextureCoordinate() {uTiling=2.2}
  VectorParameter() {group=Color, parameterName=OLD_ColorC, defaultValue=<rgba 0.4 0.3177 0.02 1>}
  VectorParameter() {group=Color, parameterName=OLD_ColorB, defaultValue=<rgba 0.5 0.4428 0.01 1>}
  VectorParameter() {group=Color, parameterName=OLD_ColorA, defaultValue=<rgba 1.2 1.0628 0.024 1>}
  VertexNormalWS()
 ]
}
