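# Idle Niagara effect for the BoldDormRise character cosmetic: a flaming helmet plume with glow, light and sparks, attached to a skeletal-mesh bone.
# Mesh core flames and licks, sprite flames and sparks surround a spring-driven pendulum ribbon plume, with platform-dependent tightness.
# System PlumeLength scales the flames and plume; AlphaADS and head-hide values dim or hide parts.
# User.Variant blends warm orange to white-grey; BasicGlow is disabled.
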NiagaraSystem /BRCosmetics/Characters/Player/Male/Medium/Bodies/M_MED_BoldDormRise_Frock/FX/NS_Character_BoldDormRise_Idle
dependencies {
 usedDataInterfaces [
  &NiagaraDataInterfacePlatformSet
  &NiagaraDataInterfaceCurve
  &NiagaraDataInterfaceColorCurve
  &NiagaraDataInterfaceSpriteRendererInfo
  &NiagaraDataInterfaceSkeletalMesh
 ]
 usedDynamicInputs [
  &Boolean_NotOperation
  &ClampFloat
  &Boolean_AndOperation
  &SetBoolByFloatComparison
 ]
 usedModules [
  &self:SystemSpawnScript.SystemScriptSource.SystemScriptGraph.NiagaraNodeAssignment_0.SetVariables_#1
  &SystemState
  &Timeline
  &self:SystemSpawnScript.SystemScriptSource.SystemScriptGraph.NiagaraNodeAssignment_2.SetVariables_#2
  &FN_VariantManager
  &self:SystemSpawnScript.SystemScriptSource.SystemScriptGraph.NiagaraNodeAssignment_1.SetVariables_#3
  &self:SystemSpawnScript.SystemScriptSource.SystemScriptGraph.NiagaraNodeAssignment_4.SetVariables_#4
  &self:SystemSpawnScript.SystemScriptSource.SystemScriptGraph.NiagaraNodeAssignment_3.SetVariables_#5
  &self:SystemSpawnScript.SystemScriptSource.SystemScriptGraph.NiagaraNodeAssignment_5.SetVariables_#6
 ]
}
emitters {
 BasicGlow {
  bEnabled=false
  emitterUpdateScript [
   &EmitterState(Life Cycle Mode=::NewEnumerator1, Loop Behavior=::NewEnumerator1, Loop Duration Mode=::NewEnumerator1, UseLoopDelay=false)
   &SpawnBurst_Instantaneous(Use Spawn Probability=false, UseLoopCountLimit=false)
  ]
  particleSpawnScript [
   &InitializeParticle(
    Color=<rgba 1 0.504 0.07 0.2>
    Sprite Size Mode=::NewEnumerator3
    Uniform Sprite Size=50
    UsePositionOffset=false
    Write Material Random=true
   )
   &CameraOffset(Camera Offset Amount=15)
  ]
  particleUpdateScript [
   &SkeletalMeshLocation(
    Bone / Socket Index=2
    Bone Sampling Mode=::NewEnumerator4
    MaskSpawnGroup=false
    Sampled Position Offset=<18 -6 0>
    Skeletal Mesh=$System.Skeletal Mesh
   )
   &ParticleState(Kill Particles When Lifetime Has Elapsed=false, Let Infinitely Lived Particles Die When Emitter Deactivates=true)
   &ScaleColor(Scale Alpha=3, ScaleA=true, ScaleRGB=true)
  ]
  rendererClasses [&NiagaraSpriteRendererProperties]
  renderers [{rendererClass=&NiagaraSpriteRendererProperties}]
 }
 BasicSparks {
  emitterUpdateScript [
   &EmitterState()
   &SpawnRate(
    SpawnRate=&MakeCustomFloatFromBool(Bool=$System.showHeadFX, True Float=10, False Float=0)
    Use Spawn Probability=false
   )
  ]
  particleSpawnScript [
   &InitializeParticle(
    Color=&LerpLinearColors(
     StartColor=&RandomRangeLinearColor(Minimum=<rgba 1 0.6933 0.2 1>, Maximum=<rgba 1 0.7625 0.05 1>)
     EndColor=&RandomRangeLinearColor(Minimum=<rgba 1 1 1 1>, Maximum=<rgba 0.7 0.7 0.7 1>)
     LerpFactor=$User.Variant
    )
    Lifetime=&RandomRangeFloat(Minimum=0.35, Maximum=1.1)
    Sprite Size=&Multiply_Vector2D_ByFloat(
     Float=&RandomRangeFloat(Minimum=0.4, Maximum=0.6)
     Vector2D=&RandomRangeVector2D(Minimum=<1 2>, Maximum=<1.3 4>)
    )
    Sprite Size Mode=::NewEnumerator4
    UsePositionOffset=false
    Write Material Random=true
   )
   &SkeletalMeshLocation(
    Bone / Socket Index=2
    Bone Sampling Mode=::NewEnumerator4
    MaskSpawnGroup=false
    Sampled Position Offset=<17 -5 0>
    Skeletal Mesh=$System.Skeletal Mesh
   )
   &SubUVAnimation(
    "Sprite Renderer "=DataInterface() {propertyValues {SpriteRenderer=&self:BasicSparks_0.NiagaraSpriteRendererProperties_0}}
    SubUV Animation Mode:ENiagara_SubUVLookupModeV2=::NewEnumerator2
    UseEndFrame=false
    UseStartFrame=false
   ) {enabled=false}
   &ShapeLocation(MaskSpawnGroup=false, Override Seed=false, Sphere Radius=8, Sphere Surface Distribution=0.2143, UseOwnerScale=true)
   &AddVelocity(
    Module.Use Falloff Distance=false
    Origin Offset=<0 -2 12>
    Override Seed=false
    Velocity Mode=::NewEnumerator1
    Velocity Origin=$Particles.SkeletalMeshLocation.SampledPosition
    Velocity Speed=&RandomRangeFloat(Minimum=30, Maximum=70)
   )
   &InheritVelocity(Inherited Velocity Amount Scale=&VectorFromFloat(Value=0.4), Speed Limit=true, UseSpeedThreshold=false) {enabled=false}
  ]
  particleUpdateScript [
   &ParticleState()
   &ScaleSpriteSize(Uniform Curve Sprite Scale=DataInterface() {propertyValues {Curve {keys [
        {interpMode=RCIM_Cubic, tangentMode=RCTM_User, time=0, value=0}
        {interpMode=RCIM_Cubic, tangentMode=RCTM_User, time=0.2, value=1}
        {interpMode=RCIM_Cubic, tangentMode=RCTM_User, time=1, value=0}
       ]}}})
   &InheritSourceMovement(
    Applied Position Scale=&VectorFromFloat(Value=&FloatFromCurve(
      FloatCurve=DataInterface() {propertyValues {Curve {keys [
          {interpMode=RCIM_Cubic, tangentMode=RCTM_User, time=0.45, value=1}
          {arriveTangent=-2.6042, interpMode=RCIM_Cubic, leaveTangent=-2.6042, tangentMode=RCTM_User, time=0.9, value=0}
         ]}}}
      Scale Curve=&RandomRangeFloat(Minimum=0.9, Maximum=1)
     ))
    ApplyForce=false
    ApplyPosition=true
    ApplyPositionFalloff=false
   )
   &CurlNoiseForce(
    Bias Noise=false
    Noise Frequency=&RandomRangeFloat(Minimum=60, Maximum=90)
    Noise Strength=&Multiply_Float(
     A=&FloatFromCurve(
      FloatCurve=DataInterface() {propertyValues {Curve {keys [{arriveTangent=2.3913, interpMode=RCIM_Cubic, leaveTangent=2.3913, tangentMode=RCTM_User, time=0, value=0.4}, <at 0.3 1>]}}}
      Scale Curve=&RandomRangeFloat(Minimum=120, Maximum=200)
     )
     B=1
    )
    Pan Noise=true
    Pan Noise Field=<0.1 0 0>
   )
   &Drag(Drag=&RandomRangeFloat(Minimum=2, Maximum=5), Use Linear Drag=true, Use Rotational Drag=true)
   &SolveForcesAndVelocity(Clamp Velocity=false, Limit Acceleration=false)
   &ScaleSpriteSizeBySpeed(Max Scale Factor=<0.8 1>, Min Scale Factor=<1 0.5>, Velocity Threshold=40)
   &DynamicMaterialParameters(
    Index 0 Param 1=0.1
    Index 0 Param 2=4
    Module.Index0Float0Write=true
    Module.Index0Float1Write=true
    Module.Index0Float2Write=false
    Module.Index0Float3Write=false
   )
   &ScaleColor(
    Scale Alpha=&Lerp_Float(A=1, B=0.025, Alpha=$System.AlphaADS)
    Scale RGB=&MakeVectorFromLinearColorRGB(Color=&LerpLinearColors(
      StartColor=&ColorFromCurve(ColorCurve=DataInterface() {propertyValues {
         AlphaCurve {keys [<at 0 1 RCIM_Cubic>]}
         BlueCurve {keys [<at 0.0943 7 RCIM_Cubic>,<at 0.3611 2.0955>,<at 1 0.25>]}
         GreenCurve {keys [<at 0.0943 8.5 RCIM_Cubic>,<at 0.3611 2.547>,<at 1 0.3583>]}
         RedCurve {keys [<at 0.0943 10 RCIM_Cubic>,<at 0.3611 3>,<at 1 0.5>]}
        }})
      EndColor=&ColorFromCurve(ColorCurve=DataInterface() {propertyValues {
         AlphaCurve {keys [<at 0 1 RCIM_Cubic>]}
         BlueCurve {keys [<at 0.0943 10 RCIM_Cubic>,<at 0.3722 3>,<at 1 0.2>]}
         GreenCurve {keys [<at 0.0943 10 RCIM_Cubic>,<at 0.3722 3>,<at 1 0.2>]}
         RedCurve {keys [<at 0.0943 10 RCIM_Cubic>,<at 0.3722 3>,<at 1 0.2>]}
        }})
      LerpFactor=$User.Variant
     ))
    ScaleA=true
    ScaleRGB=true
   )
   /SetVariables(
    Particles.ColorGlow:LinearColor=&MultiplyLinearColor(
     Linear Color=&LerpLinearColors(StartColor=<rgba 0.15 0.125 0.075 0.2>, EndColor=<rgba 0.1 0.1 0.1 0.2>, LerpFactor=$User.Variant)
     Scale Factor=&MakeVector4FromVectorAndFloat(
      "Vector (RGB)"=&MakeVectorFromLinearColorRGB(Color=$Particles.Color)
      "Float (Alpha)"=&Lerp_Float(A=1, B=1, Alpha=$System.AlphaADS)
     )
    )
    Particles.DynParGlow:Vector4f=<1 1 1 0>
    Particles.SpriteSizeGlow:Vector2f=&Multiply_Vector2D(A=$Particles.SpriteSize, B=<8 4>)
   ) {isSetParameters=true}
  ]
  rendererClasses [&NiagaraSpriteRendererProperties]
  renderers [
   {rendererClass=&NiagaraSpriteRendererProperties}
   {rendererClass=&NiagaraSpriteRendererProperties, rendererIndex=1}
  ]
 }
 CoreFlames {
  emitterUpdateScript [
   &EmitterState(Life Cycle Mode=::NewEnumerator1, Loop Behavior=::NewEnumerator1, Loop Duration Mode=::NewEnumerator1, UseLoopDelay=false)
   &SpawnBurst_Instantaneous(Use Spawn Probability=false, UseLoopCountLimit=true)
  ]
  particleSpawnScript [
   &InitializeParticle(
    Mesh Scale=&Multiply_VectorByFloat(Vector=<1 1 1>, Float=0.07)
    Mesh Scale Mode=::NewEnumerator4
    UsePositionOffset=false
    Write Material Random=true
   )
   &InitialMeshOrientation(Mesh Axis To Orient=<0 1 0>, Mesh Orientation Mode=::NewEnumerator2, Rotation=<0.5 0 0>, Use Rotation Vector=true)
   &CameraOffset(Camera Offset Amount=2)
   &InitialMeshOrientation(Mesh Orientation Mode=::NewEnumerator2, Rotation=<0 0 -0.025>, Use Rotation Vector=true) {module=InitialMeshOrientation001}
  ]
  particleUpdateScript [
   &SkeletalMeshLocation(
    Bone / Socket Index=2
    Bone Sampling Mode=::NewEnumerator4
    MaskSpawnGroup=false
    Sampled Position Offset=<14 1 0>
    Skeletal Mesh=$System.Skeletal Mesh
   )
   &ParticleState(Kill Particles When Lifetime Has Elapsed=false, Let Infinitely Lived Particles Die When Emitter Deactivates=true)
   &DynamicMaterialParameters(
    Index 0 Param 1=50
    Index 0 Param 4=0.15
    Module.Index0Float0Write=true
    Module.Index0Float1Write=false
    Module.Index0Float2Write=false
    Module.Index0Float3Write=true
   )
   &ScaleColor(
    Scale Alpha=&Multiply_Float(A=$System.FadeInAfterHideHead, B=&Lerp_Float(A=1, B=2, Alpha=$User.Variant))
    Scale RGB=&VectorFromFloat(Value=1.3)
    ScaleA=true
    ScaleRGB=true
   )
   /SetVariables(Particles.ColorGlow:LinearColor=&MultiplyLinearColor(
     Linear Color=&LerpLinearColors(StartColor=<rgba 1 0.4987 0.06 0.22>, EndColor=<rgba 0.66 0.66 0.66 0.16>, LerpFactor=$User.Variant)
     Scale Factor=<1 1 1 1>
    )) {isSetParameters=true}
   /SetVariables(Particles.DynParGlow:Vector4f=&Multiply_Vector4(A=$Particles.DynamicMaterialParameter, B=<1 1 1 2.2>)) {isSetParameters=true}
   /SetVariables(Particles.DynParCore:Vector4f=&Multiply_Vector4(A=$Particles.DynamicMaterialParameter, B=<1 1 1 0.8>)) {isSetParameters=true}
   /SetVariables(Particles.ScaleCore:Vector3f=&Multiply_Vector(A=$Particles.Scale, B=&Multiply_VectorByFloat(Vector=<1 1 1>, Float=1))) {isSetParameters=true}
   /SetVariables(Particles.ColorCore:LinearColor=<rgba 10 10 10 1>) {isSetParameters=true}
  ]
  rendererClasses [&NiagaraMeshRendererProperties]
  renderers [
   {rendererClass=&NiagaraMeshRendererProperties, rendererIndex=0}
   {rendererClass=&NiagaraMeshRendererProperties, rendererIndex=1}
   {rendererClass=&NiagaraMeshRendererProperties, rendererIndex=2}
   {rendererClass=&NiagaraMeshRendererProperties, rendererIndex=3}
  ]
 }
 Flame_Licks {
  emitterUpdateScript [
   &EmitterState(Life Cycle Mode=::NewEnumerator1, Loop Behavior=::NewEnumerator1, Loop Duration Mode=::NewEnumerator1, UseLoopDelay=false)
   &SpawnBurst_Instantaneous(Spawn Count=4, Use Spawn Probability=false, UseLoopCountLimit=true)
  ]
  particleSpawnScript [
   &PartitionParticles(Number Of Partitions=4)
   &InitializeParticle(
    Color=<rgba 1.5 1.5 1.5 1>
    Mesh Scale=&Multiply_VectorByFloat(
     Vector=<1 1 1>
     Float=&MergeHelper_FloatSwitch(0=0.19, 1=0.19, 2=0.15, 3=0.15, Index=$Particles.PartitionParticles.Partition, Num=::NewEnumerator2)
    )
    Mesh Scale Mode=::NewEnumerator4
    UsePositionOffset=false
    Write Material Random=true
   )
   &InitialMeshOrientation(Mesh Axis To Orient=<0 1 0>, Mesh Orientation Mode=::NewEnumerator2, Rotation=<0.5 0 0>, Use Rotation Vector=true)
   &CameraOffset(Camera Offset Amount=2)
   &InitialMeshOrientation(
    Mesh Orientation Mode=::NewEnumerator2
    Rotation=&MergeHelper_VectorSwitch(
     0=<0.03 0 -0.04>
     1=<-0.03 0 -0.04>
     2=<0.03 0 -0.09>
     3=<-0.03 0 -0.09>
     Index=$Particles.PartitionParticles.Partition
     Num=::NewEnumerator2
    )
    Use Rotation Vector=true
   ) {module=InitialMeshOrientation001}
  ]
  particleUpdateScript [
   &SkeletalMeshLocation(
    Bone / Socket Index=2
    Bone Sampling Mode=::NewEnumerator4
    MaskSpawnGroup=false
    Sampled Position Offset=&AddVectorToPosition(
     Position=<15 0 0>
     Vector=&MergeHelper_VectorSwitch(0=<0 0 -1>, 1=<0 0 1>, 2=<-2 -4 -2.5>, 3=<-2 -4 2.5>, Index=$Particles.PartitionParticles.Partition, Num=::NewEnumerator2)
    )
    Skeletal Mesh=$System.Skeletal Mesh
   )
   &ParticleState(Kill Particles When Lifetime Has Elapsed=false, Let Infinitely Lived Particles Die When Emitter Deactivates=true)
   &ScaleMeshSize(Scale Factor=&VectorFromFloat(Value=&Lerp_Float(
      A=0.8
      B=1.1
      Alpha=&FloatFromCurve(
       FloatCurve=DataInterface() {propertyValues {Curve {keys [{arriveTangent=2.2691, interpMode=RCIM_Cubic, leaveTangent=2.2691, tangentMode=RCTM_User, time=0, value=0}, <at 0.8 1>]}}}
       CurveIndex=$System.PlumeLength
      )
     )))
   &ScaleColor(Scale Alpha=$System.FadeInAfterHideHead, ScaleA=true, ScaleRGB=true)
   &DynamicMaterialParameters(
    Index 0 Param 1=&RandomRangeFloat(Minimum=18, Maximum=24)
    Index 0 Param 2=&RandomRangeFloat(Minimum=0, Maximum=1)
    Index 0 Param 4=&Multiply_Float(
     A=&MergeHelper_FloatSwitch(0=1, 1=1, 2=1.1, 3=1.1, Index=$Particles.PartitionParticles.Partition, Num=::NewEnumerator2)
     B=&Lerp_Float(A=0.17, B=0.22, Alpha=$System.PlumeLength)
    )
    Module.Index0Float0Write=true
    Module.Index0Float1Write=true
    Module.Index0Float2Write=false
    Module.Index0Float3Write=true
   )
   /SetVariables(Particles.ColorGlow:LinearColor=&MultiplyLinearColor(
     Linear Color=&LerpLinearColors(StartColor=<rgba 1 0.4987 0.06 0.06>, EndColor=<rgba 0.66 0.66 0.66 0.1>, LerpFactor=$User.Variant)
     Scale Factor=<1 1 1 1>
    )) {isSetParameters=true}
   /SetVariables(Particles.DynParGlow:Vector4f=&Multiply_Vector4(A=$Particles.DynamicMaterialParameter, B=<1 1 1 1>)) {isSetParameters=true}
  ]
  rendererClasses [&NiagaraMeshRendererProperties]
  renderers [
   {rendererClass=&NiagaraMeshRendererProperties, rendererIndex=0}
   {rendererClass=&NiagaraMeshRendererProperties, rendererIndex=1}
  ]
 }
 Flames {
  emitterUpdateScript [
   &EmitterState(Life Cycle Mode=::NewEnumerator1, Loop Behavior=::NewEnumerator1, Loop Duration Mode=::NewEnumerator1, UseLoopDelay=false)
   &SpawnRate(SpawnRate=&Lerp_Float(A=7, B=15, Alpha=$System.PlumeLength), Use Spawn Probability=false)
  ]
  particleSpawnScript [
   /SetVariables(Particles.randomValue:NiagaraFloat=&RandomRangeFloat(Minimum=0, Maximum=1)) {isSetParameters=true}
   &InitializeParticle(
    Lifetime=&RandomRangeFloat(Minimum=0.6, Maximum=0.8)
    Sprite Rotation Mode=::NewEnumerator2
    Sprite Size Mode=::NewEnumerator3
    Uniform Sprite Size=&Multiply_Float(
     A=&RandomRangeFloat(Minimum=27, Maximum=40)
     B=&Lerp_Float(A=0.85, B=1, Alpha=$System.PlumeLength)
    )
    UsePositionOffset=false
    Write Material Random=true
   )
   &CameraOffset(Camera Offset Amount=8)
  ]
  particleUpdateScript [
   &SkeletalMeshLocation(
    Bone / Socket Index=2
    Bone Sampling Mode=::NewEnumerator4
    MaskSpawnGroup=false
    Orientation Sampling=::NewEnumerator1
    Sampled Position Offset=&AddVectorToPosition(
     Position=<18 -3 0>
     Vector=&Lerp_Vector(
      A=<0 0 0>
      B=&MakeVector(
       X=&FloatFromCurve(
        FloatCurve=DataInterface() {propertyValues {Curve {keys [
            {arriveTangent=0.1858, interpMode=RCIM_Cubic, tangentMode=RCTM_Break, time=0, value=0}
            {interpMode=RCIM_Cubic, tangentMode=RCTM_Break, time=1, value=1}
           ]}}}
        CurveIndex=$Particles.randomValue
        Scale Curve=-25
       )
       Y=&FloatFromCurve(
        FloatCurve=DataInterface() {propertyValues {Curve {keys [
            {arriveTangent=3.1737, interpMode=RCIM_Cubic, leaveTangent=3.1737, tangentMode=RCTM_User, time=0, value=0}
            {arriveTangent=0.174, interpMode=RCIM_Cubic, leaveTangent=0.174, tangentMode=RCTM_User, time=1, value=1}
           ]}}}
        CurveIndex=$Particles.randomValue
        Scale Curve=-23
       )
       Z=0
      )
      Alpha=$System.PlumeLength
     )
    )
    Skeletal Mesh=$System.Skeletal Mesh
   )
   &ParticleState()
   &ScaleSpriteSize(
    Scale Sprite Size Mode=::NewEnumerator0
    Uniform Scale Factor=&Lerp_Float(A=1, B=&Lerp_Float(A=1, B=0.5, Alpha=$Particles.randomValue), Alpha=$System.PlumeLength)
   ) {module=ScaleSpriteSize001}
   &ScaleSpriteSize(Uniform Curve Sprite Scale=DataInterface() {propertyValues {Curve {keys [
        {arriveTangent=2.4341, interpMode=RCIM_Cubic, leaveTangent=2.4341, tangentMode=RCTM_User, time=0, value=0.1}
        {interpMode=RCIM_Cubic, tangentMode=RCTM_User, time=0.3, value=1}
        {arriveTangent=-1.1638, interpMode=RCIM_Cubic, leaveTangent=-1.1638, tangentMode=RCTM_User, time=1, value=0.3}
       ]}}})
   &ScaleColor(
    Scale Alpha=&FloatFromCurve(
     FloatCurve=DataInterface() {propertyValues {Curve {keys [
         {interpMode=RCIM_Cubic, tangentMode=RCTM_User, time=0, value=0}
         {arriveTangent=-1.6196, interpMode=RCIM_Cubic, leaveTangent=-1.6196, tangentMode=RCTM_User, time=0.4, value=1}
         <at 1 0>
        ]}}}
     Scale Curve=&Multiply_Float(
      A=&Lerp_Float(A=1, B=1.5, Alpha=$User.Variant)
      B=&RandomRangeFloat(Minimum=0.2, Maximum=0.4)
     )
    )
    Scale RGB=&MakeVectorFromLinearColorRGB(Color=&LerpLinearColors(
      StartColor=&ColorFromCurve(ColorCurve=DataInterface() {propertyValues {
         AlphaCurve {keys [<at 0 1 RCIM_Cubic>]}
         BlueCurve {keys [<at 0.1076 0.825 RCIM_Cubic>,<at 0.3438 0.5217>,<at 0.8891 0.3>]}
         GreenCurve {keys [<at 0.1076 1.1962 RCIM_Cubic>,<at 0.3438 0.7676>,<at 0.8891 0.5333>]}
         RedCurve {keys [<at 0.1076 1.5 RCIM_Cubic>,<at 0.3438 1>,<at 0.8891 1>]}
        }})
      EndColor=&ColorFromCurve(ColorCurve=DataInterface() {propertyValues {
         AlphaCurve {keys [<at 0 1 RCIM_Cubic>]}
         BlueCurve {keys [<at 0.1705 0.03 RCIM_Cubic>,<at 0.8934 0.4>]}
         GreenCurve {keys [<at 0.1705 0.03 RCIM_Cubic>,<at 0.8934 0.4>]}
         RedCurve {keys [<at 0.1705 0.03 RCIM_Cubic>,<at 0.8934 0.4>]}
        }})
      LerpFactor=$User.Variant
     ))
    ScaleA=true
    ScaleRGB=true
   )
   &AccelerationForce(Acceleration=&RandomRangeVector(Minimum=<0 0 30>, Maximum=<0 0 70>)) {enabled=false}
   &Drag(Drag=&RandomRangeFloat(Minimum=1, Maximum=2.5), Use Linear Drag=true, Use Rotational Drag=true) {enabled=false}
   &SolveForcesAndVelocity(Clamp Velocity=false, Limit Acceleration=false) {enabled=false}
   &DynamicMaterialParameters(
    Index 0 Param 1=&FloatFromCurve(FloatCurve=DataInterface() {propertyValues {Curve {keys [<at 0 0>,<at 1 1>]}}}, Scale Curve=0.66)
    Module.Index0Float0Write=true
    Module.Index0Float1Write=false
    Module.Index0Float2Write=false
    Module.Index0Float3Write=false
   )
   &ScaleColor(
    Scale Alpha=&FloatFromCurve(
     FloatCurve=DataInterface() {propertyValues {Curve {keys [{interpMode=RCIM_Cubic, tangentMode=RCTM_User, time=0.2, value=1}, {interpMode=RCIM_Cubic, tangentMode=RCTM_User, time=0.5, value=0.1}]}}}
     CurveIndex=$Particles.randomValue
    )
    ScaleA=true
    ScaleRGB=true
   ) {module=ScaleColor001}
  ]
  rendererClasses [&NiagaraSpriteRendererProperties]
  renderers [
   {rendererClass=&NiagaraSpriteRendererProperties}
   {rendererClass=&NiagaraSpriteRendererProperties, rendererIndex=1}
  ]
 }
 HelmetGlow {
  emitterUpdateScript [
   &EmitterState(Life Cycle Mode=::NewEnumerator1, Loop Behavior=::NewEnumerator1, Loop Duration Mode=::NewEnumerator1, UseLoopDelay=false)
   &SpawnBurst_Instantaneous(Use Spawn Probability=false, UseLoopCountLimit=true)
  ]
  particleSpawnScript [
   &InitializeParticle(
    Mesh Scale=<0.6 0.6 0.5>
    Mesh Scale Mode=::NewEnumerator4
    Sprite Size Mode=::NewEnumerator3
    Uniform Sprite Size=30
    UsePositionOffset=false
    Write Material Random=true
   )
   &CameraOffset(Camera Offset Amount=3)
   &InitialMeshOrientation(Mesh Axis To Orient=<0 1 0>, Mesh Orientation Mode=::NewEnumerator2, Rotation=<-0.25 0 0>, Use Rotation Vector=true)
   /SetVariables(Particles.ColorFlame:LinearColor=&LerpLinearColors(StartColor=<rgba 3 2 1.5 1>, EndColor=<rgba 1 1 1 1>, LerpFactor=$User.Variant)) {enabled=false, isSetParameters=true}
  ]
  particleUpdateScript [
   &SkeletalMeshLocation(
    Bone / Socket Index=2
    Bone Sampling Mode=::NewEnumerator4
    MaskSpawnGroup=false
    Sampled Position Offset=<13 -4 0>
    Skeletal Mesh=$System.Skeletal Mesh
   )
   /SetVariables(
    Particles.ColorGlow:LinearColor=&Multiply_LinearColorByFloat(
     Linear Color=&LerpLinearColors(
      StartColor=<rgba 1 0.4575 0.07 0.6>
      EndColor=&MakeCustomLinearColorFromBool(Bool=$User.bIsFrontEndPreview, True Color=<rgba 0.66 0.66 0.66 0.25>, False Color=<rgba 0.66 0.66 0.66 0.5>)
      LerpFactor=$User.Variant
     )
     Float=&Multiply_Float(
      A=&FNTimeOfDayScalar(Night Scale=0.66, Morning Scale=0.8, Evening Scale=0.8)
      B=&Multiply_Float(A=$System.FadeInAfterHideHead, B=&Lerp_Float(A=1, B=0.33, Alpha=$System.AlphaADS))
     )
    )
    Particles.PositionGlow:NiagaraPosition=&AddVectorToPosition(
     Position=$Particles.Position
     Vector=&RotateVectorByQuaternion(Vector To Rotate=<4 -2 0>, Quaternion=$Particles.SkeletalMeshLocation.SampledRotation)
    )
   ) {isSetParameters=true}
   &ParticleState(Kill Particles When Lifetime Has Elapsed=false, Let Infinitely Lived Particles Die When Emitter Deactivates=true)
   &ScaleSpriteSize(Scale Sprite Size Mode=::NewEnumerator0, Uniform Scale Factor=&Lerp_Float(A=1, B=1.25, Alpha=$System.PlumeLength))
   /SetVariables(
    Particles.CameraOffsetGlow=9
    Particles.SpriteSizeGlow:Vector2f=&Multiply_Vector2D_ByFloat(Float=1.66, Vector2D=$Particles.SpriteSize)
   ) {isSetParameters=true}
   &ScaleColor(
    Scale Alpha=&Multiply_Float(A=$System.FadeInAfterHideHead, B=&Lerp_Float(A=1, B=0.33, Alpha=$System.AlphaADS))
    Scale RGB=&MakeVectorFromLinearColorRGB(Color=&LerpLinearColors(StartColor=<rgba 25 7.5 0 1>, EndColor=<rgba 5 5 5 1>, LerpFactor=$User.Variant))
    ScaleA=true
    ScaleRGB=true
   )
   &FN_Light_Attributes(
    Color=&LerpLinearColors(StartColor=<rgba 36 16.47 2.52 1>, EndColor=<rgba 14 14 14 1>, LerpFactor=$User.Variant)
    Create Color Attribute=true
    Scale Color=&Waveform(
     Waveform Count=::NewEnumerator1
     "Waveform [1]"=::NewEnumerator0
     "[1] Input Position"=0.1
     "[1] Frequency"=0.433
     "[1] Amplitude Min/Max"=<0.4 1>
     "Waveform [2] Blend"=::NewEnumerator4
     "[2] Input Position"=0
     "[2] Phase"=$Emitter.Age
     "[2] Phase Scale"=1
     "[2] Compound Frequency"=2
     "[2] Frequency"=3.1
     "[2] Amplitude Min/Max"=<0.8 1>
     "[2] Scale"=1
     "[2] Bias"=0
     "[2] Exponent"=1
     "[2] Pulse Duration"=0.1
     "[2] Clip Triangle Min/Max"=<0 0>
     "[2] Random Seed"=0
    )
   )
   &DynamicMaterialParameters(
    Index 0 Param 3=0.5
    Index 1 Param 4=0.5
    Module.Index0Float0Write=false
    Module.Index0Float1Write=false
    Module.Index0Float2Write=true
    Module.Index0Float3Write=false
    Module.Index1Float0Write=true
    Module.Index1Float1Write=true
    Module.Index1Float2Write=true
    Module.Index1Float3Write=true
    Module.Index2Float0Write=false
    Module.Index2Float1Write=false
    Module.Index2Float2Write=false
    Module.Index2Float3Write=false
    Write Parameter Index 1=::NewEnumerator1
    Write Parameter Index 2=::NewEnumerator1
   )
   /SetVariables(
    Particles.ColorFlame:LinearColor=&MultiplyLinearColor(
     Linear Color=&LerpLinearColors(StartColor=<rgba 3 2 1.5 1>, EndColor=<rgba 1 1 1 1>, LerpFactor=$User.Variant)
     Scale Factor=&MakeVector4FromVectorAndFloat(
      "Vector (RGB)"=&VectorFromFloat(Value=&Lerp_Float(A=1, B=1, Alpha=$User.Variant))
      "Float (Alpha)"=&Multiply_Float(A=$System.FadeInAfterHideHead, B=&Lerp_Float(A=0.02, B=0.2, Alpha=$User.Variant))
     )
    )
    Particles.PositionFlame:NiagaraPosition=&AddVectorToPosition(
     Position=$Particles.Position
     Vector=&RotateVectorByQuaternion(Vector To Rotate=<1 5 0>, Quaternion=$Particles.SkeletalMeshLocation.SampledRotation)
    )
   ) {isSetParameters=true}
  ]
  rendererClasses [
   &NiagaraSpriteRendererProperties
   &NiagaraLightRendererProperties
   &NiagaraMeshRendererProperties
  ]
  renderers [
   {rendererClass=&NiagaraSpriteRendererProperties}
   {rendererClass=&NiagaraSpriteRendererProperties, rendererIndex=1}
   {rendererClass=&NiagaraLightRendererProperties, rendererIndex=2}
   {rendererClass=&NiagaraMeshRendererProperties, rendererIndex=3}
   {rendererClass=&NiagaraMeshRendererProperties, rendererIndex=4}
  ]
 }
 Plume_Flame {
  emitterUpdateScript [
   &EmitterState(Life Cycle Mode=::NewEnumerator1, Loop Behavior=::NewEnumerator1, Loop Duration Mode=::NewEnumerator1, UseLoopDelay=false)
   /SetVariables(Emitter.ManeParticleCount=32) {isSetParameters=true}
   &SpawnBurst_Instantaneous(Spawn Count=$Emitter.ManeParticleCount, Use Spawn Probability=false, UseLoopCountLimit=true)
   /SetVariables(Emitter.PlumeTightnessByPlatform:NiagaraFloat=&FN_PickFloatDependOnPlatform(TargetPlatform=DataInterface() {propertyValues {Platforms {qualityLevelMask=-8}}}, TargetPlatformActive=2, TargetPlatformNoActive=4)) {isSetParameters=true}
  ]
  particleSpawnScript [
   &InitializeParticle(
    Lifetime=0.5
    Ribbon Width=25
    Ribbon Width Mode=::NewEnumerator1
    Sprite Size Mode=::NewEnumerator3
    Uniform Sprite Size=1
    UsePositionOffset=false
    Write Material Random=true
   )
   &PartitionParticles(
    Number Of Partitions=1
    Particles Per Partition:NiagaraInt32=$Emitter.ManeParticleCount
    Partition Mode:ENiagara_PartitionMode_v2=::NewEnumerator1
   )
   /SetVariables(Particles.NormalizedChainPosition:NiagaraFloat=&Lerp_Float(
     A=1
     B=0
     Alpha=&FloatFromCurve(
      FloatCurve=DataInterface() {propertyValues {Curve {keys [
          {interpMode=RCIM_Cubic, tangentMode=RCTM_User, time=0, value=0}
          {arriveTangent=1.4974, interpMode=RCIM_Cubic, leaveTangent=1.4974, tangentMode=RCTM_User, time=1, value=1}
         ]}}}
      CurveIndex=$Particles.PartitionParticles.Normalized Index In Partition
     )
    )) {isSetParameters=true}
  ]
  particleUpdateScript [
   &SkeletalMeshLocation(
    Bone / Socket Index=2
    Bone Sampling Mode=::NewEnumerator4
    MaskSpawnGroup=false
    Position Sampling=::NewEnumerator2
    Sampled Position Offset=<15 6 0>
    Skeletal Mesh=$System.Skeletal Mesh
   )
   /SetVariables(Particles.RibbonFacing=&RotateVectorByQuaternion(Vector To Rotate=<0 0 1>, Quaternion=$Particles.SkeletalMeshLocation.SampledRotation)) {isSetParameters=true}
   &PendulumSetup(
    Module.Enable Spring Driver=true
    Pendulum Length:NiagaraFloat=&Lerp_Float(A=0, B=60, Alpha=$Particles.NormalizedChainPosition)
    Pendulum Pivot:NiagaraPosition=$Particles.SkeletalMeshLocation.SampledPosition
    Pendulum Rest Axis:Vector3f=&RotateVectorByQuaternion(
     Vector To Rotate=&MakeVector(
      X=&FloatFromCurve(
       FloatCurve=DataInterface() {propertyValues {Curve {keys [<at 0 -0.1>,<at 0.25 1>,<at 1 -1>]}}}
       CurveIndex=$Particles.NormalizedChainPosition
      )
      Y=&Lerp_Float(
       A=-5
       B=-2
       Alpha=&FloatFromCurve(FloatCurve=DataInterface() {propertyValues {Curve {keys [<at 0 0>,<at 1 1>]}}}, CurveIndex=$Particles.NormalizedChainPosition)
      )
      Z=0
     )
     Quaternion=$Particles.SkeletalMeshLocation.SampledRotation
    )
    "Spring Driven Constraint (Tightness)":NiagaraFloat=&Lerp_Float(A=60, B=10, Alpha=$Particles.NormalizedChainPosition)
   ) {enabled=false, module=PendulumSetup001}
   &PendulumSetup(
    Module.Enable Spring Driver=true
    Pendulum Length:NiagaraFloat=&Lerp_Float(A=0, B=&Lerp_Float(A=30, B=75, Alpha=$System.PlumeLength), Alpha=$Particles.NormalizedChainPosition)
    Pendulum Pivot:Vector3f=&AddVectorToPosition(Position=$Particles.SkeletalMeshLocation.SampledPosition, Vector=<0 0 0>)
    Pendulum Rest Axis:Vector3f=&RotateVectorByQuaternion(
     Vector To Rotate=&MakeVector(
      X=&FloatFromCurve(
       FloatCurve=DataInterface() {propertyValues {Curve {keys [
           {arriveTangent=-0.6298, interpMode=RCIM_Cubic, leaveTangent=-0.6298, tangentMode=RCTM_User, time=0, value=0}
           {arriveTangent=-1.1324, interpMode=RCIM_Cubic, leaveTangent=-1.1324, tangentMode=RCTM_User, time=1, value=-1}
          ]}}}
       CurveIndex=$Particles.NormalizedChainPosition
       Scale Curve=&Lerp_Float(A=0.1, B=1, Alpha=$System.PlumeLength)
      )
      Y=&Lerp_Float(
       A=-3
       B=-1
       Alpha=&FloatFromCurve(
        FloatCurve=DataInterface() {propertyValues {Curve {keys [
            {arriveTangent=3.2767, interpMode=RCIM_Cubic, leaveTangent=3.2767, tangentMode=RCTM_User, time=0, value=0}
            {interpMode=RCIM_Cubic, tangentMode=RCTM_User, time=0.9, value=1}
           ]}}}
        CurveIndex=$Particles.NormalizedChainPosition
        Scale Curve=&Lerp_Float(A=1.2, B=1, Alpha=$System.PlumeLength)
       )
      )
      Z=0
     )
     Quaternion=$Particles.SkeletalMeshLocation.SampledRotation
    )
    "Spring Driven Constraint (Tightness)":NiagaraFloat=&Multiply_Float(
     A=$Emitter.PlumeTightnessByPlatform
     B=&Multiply_Float(
      A=&Lerp_Float(A=1, B=2, Alpha=$System.AlphaADS)
      B=&Lerp_Float(
       A=100
       B=&Multiply_Float(
        A=&Lerp_Float(A=4, B=1.5, Alpha=$System.PlumeLength)
        B=&Lerp_Float(
         A=28
         B=9
         Alpha=&FloatFromCurve(
          FloatCurve=DataInterface() {propertyValues {Curve {keys [
              {interpMode=RCIM_Cubic, tangentMode=RCTM_User, time=0, value=0}
              {interpMode=RCIM_Cubic, tangentMode=RCTM_User, time=0.12, value=0.3}
              {arriveTangent=1.6017, interpMode=RCIM_Cubic, leaveTangent=1.6017, tangentMode=RCTM_User, time=1, value=1}
             ]}}}
          CurveIndex=$Particles.NormalizedChainPosition
         )
        )
       )
       Alpha=&FloatFromCurve(FloatCurve=DataInterface() {propertyValues {Curve {keys [<at 0 0>,<at 1 1>]}}})
      )
     )
    )
   )
   &InheritVelocity(
    Inherited Velocity Speed Limit:NiagaraFloat=&FloatFromCurve(
     FloatCurve=DataInterface() {propertyValues {Curve {keys [<at 0.1 0>,<at 0.7 1>]}}}
     CurveIndex=$Particles.NormalizedChainPosition
     Scale Curve=-300
    )
    Speed Limit=true
    UseSpeedThreshold=false
   )
   &SolveForcesAndVelocity(Clamp Velocity=false, Limit Acceleration=false)
   &ParticleState(Kill Particles When Lifetime Has Elapsed=false, Let Infinitely Lived Particles Die When Emitter Deactivates=true)
   &PendulumConstraint()
   &CalculateAccurateVelocity()
   &ScaleColor(
    Scale Alpha=&Multiply_Float(
     A=&FloatFromCurve(
      FloatCurve=DataInterface() {propertyValues {Curve {keys [
          {interpMode=RCIM_Cubic, tangentMode=RCTM_User, time=0.05, value=0}
          {interpMode=RCIM_Cubic, tangentMode=RCTM_User, time=0.25, value=1}
          {arriveTangent=-0.3263, interpMode=RCIM_Cubic, leaveTangent=-0.1049, tangentMode=RCTM_Break, time=0.3, value=1}
          {arriveTangent=-0.3357, interpMode=RCIM_Cubic, leaveTangent=-0.3357, tangentMode=RCTM_User, time=1, value=0}
         ]}}}
      CurveIndex=$Particles.NormalizedChainPosition
      Scale Curve=&Lerp_Float(
       A=&FNTimeOfDayScalar(Night Scale=1, Morning Scale=1.4, Day Scale=1.66, Evening Scale=1.4)
       B=1.75
       Alpha=$User.Variant
      )
     )
     B=&FloatFromCurve(
      FloatCurve=DataInterface() {propertyValues {Curve {keys [
          {arriveTangent=0.0878, interpMode=RCIM_Cubic, leaveTangent=0.0878, tangentMode=RCTM_User, time=0.1, value=0}
          {arriveTangent=2.054, interpMode=RCIM_Cubic, leaveTangent=2.054, tangentMode=RCTM_User, time=0.5, value=0.3}
          {interpMode=RCIM_Cubic, tangentMode=RCTM_User, time=0.8, value=1}
         ]}}}
      CurveIndex=$System.PlumeLength
     )
    )
    Scale RGB=&MakeVectorFromLinearColorRGB(Color=&LerpLinearColors(
      StartColor=<rgba 1 1 1 1>
      EndColor=&ColorFromCurve(
       ColorCurve=DataInterface() {propertyValues {
         AlphaCurve {keys [<at 0 1 RCIM_Cubic>]}
         BlueCurve {keys [<at 0.0536 1 RCIM_Cubic>,<at 0.748 0.33 RCIM_Cubic>]}
         GreenCurve {keys [<at 0.0536 1 RCIM_Cubic>,<at 0.748 0.33 RCIM_Cubic>]}
         RedCurve {keys [<at 0.0536 1 RCIM_Cubic>,<at 0.748 0.33 RCIM_Cubic>]}
        }}
       CurveIndex=$Particles.NormalizedChainPosition
      )
      LerpFactor=$User.Variant
     ))
    ScaleA=true
    ScaleRGB=true
   )
   &ScaleRibbonWidth(Ribbon Width Scale=&Multiply_Float(
     A=&FloatFromCurve(
      FloatCurve=DataInterface() {propertyValues {Curve {keys [
          {arriveTangent=7.4928, interpMode=RCIM_Cubic, leaveTangent=7.4928, tangentMode=RCTM_User, time=0, value=0.3}
          {interpMode=RCIM_Cubic, tangentMode=RCTM_User, time=0.3, value=1}
          {arriveTangent=-1.2799, interpMode=RCIM_Cubic, leaveTangent=-1.2799, tangentMode=RCTM_User, time=1, value=0.3}
         ]}}}
      CurveIndex=$Particles.NormalizedChainPosition
     )
     B=&FloatFromCurve(
      FloatCurve=DataInterface() {propertyValues {Curve {keys [
          {arriveTangent=0.3064, interpMode=RCIM_Cubic, leaveTangent=0.3064, tangentMode=RCTM_User, time=0.5, value=0.33}
          {arriveTangent=0.2817, interpMode=RCIM_Cubic, leaveTangent=0.2817, tangentMode=RCTM_User, time=0.95, value=1}
         ]}}}
      CurveIndex=$System.PlumeLength
     )
    ))
   &ScaleColor(
    Scale Alpha=&FloatFromCurve(FloatCurve=DataInterface() {propertyValues {Curve {keys [<at 0.2 0>,<at 0.3 1>]}}}, CurveIndex=$System.PlumeLength)
    Scale RGB=&VectorFromFloat(Value=&Lerp_Float(
      A=1
      B=1
      Alpha=&FloatFromCurve(
       FloatCurve=DataInterface() {propertyValues {Curve {keys [{interpMode=RCIM_Cubic, tangentMode=RCTM_User, time=0.4, value=0}, <at 0.8 1>]}}}
       CurveIndex=$System.PlumeLength
      )
     ))
    ScaleA=true
    ScaleRGB=false
   ) {module=ScaleColor001}
   &ScaleColor(
    Scale Alpha=&Lerp_Float(A=1, B=&Lerp_Float(A=0.25, B=0.66, Alpha=$User.Variant), Alpha=$System.AlphaADS)
    ScaleA=true
    ScaleRGB=false
   ) {module=ScaleColor002}
   &ScaleColor(Scale Alpha=$System.FadeInAfterHideHead, ScaleA=true, ScaleRGB=false) {module=ScaleColor003}
   /SetVariables(Particles.ColorGlow:LinearColor=&MultiplyLinearColor(
     Linear Color=&LerpLinearColors(StartColor=<rgba 1 0.4987 0.06 0.33>, EndColor=<rgba 0.3 0.3 0.3 0.12>, LerpFactor=$User.Variant)
     Scale Factor=&MakeVector4FromVectorAndFloat(
      "Vector (RGB)"=<1 1 1>
      "Float (Alpha)"=&Multiply_Float(
       A=$System.FadeInAfterHideHead
       B=&Multiply_Float(
        A=&FloatFromCurve(
         FloatCurve=DataInterface() {propertyValues {Curve {keys [
             {arriveTangent=0.0878, interpMode=RCIM_Cubic, leaveTangent=0.0878, tangentMode=RCTM_User, time=0.3, value=0}
             {arriveTangent=2.054, interpMode=RCIM_Cubic, leaveTangent=2.054, tangentMode=RCTM_User, time=0.5, value=0.3}
             {interpMode=RCIM_Cubic, tangentMode=RCTM_User, time=0.8, value=1}
            ]}}}
         CurveIndex=$System.PlumeLength
        )
        B=&FloatFromCurve(
         FloatCurve=DataInterface() {propertyValues {Curve {keys [{interpMode=RCIM_Cubic, tangentMode=RCTM_User, time=0.5, value=1}, {interpMode=RCIM_Cubic, tangentMode=RCTM_User, time=1, value=0}]}}}
         CurveIndex=$Particles.NormalizedChainPosition
         Scale Curve=&FloatFromCurve(
          FloatCurve=DataInterface() {propertyValues {Curve {keys [<at 0.2 0>,<at 0.3 1>]}}}
          CurveIndex=$System.PlumeLength
          Scale Curve=&Lerp_Float(A=1.5, B=0.7, Alpha=$System.AlphaADS)
         )
        )
       )
      )
     )
    )) {isSetParameters=true}
   /SetVariables(Particles.RibbonWidthGlow:NiagaraFloat=&Multiply_Float(
     A=$Particles.Initial.RibbonWidth
     B=&FloatFromCurve(
      FloatCurve=DataInterface() {propertyValues {Curve {keys [
          {arriveTangent=7.4928, interpMode=RCIM_Cubic, leaveTangent=7.4928, tangentMode=RCTM_User, time=0, value=0}
          {arriveTangent=0.0949, interpMode=RCIM_Cubic, leaveTangent=0.0949, tangentMode=RCTM_User, time=0.2, value=1}
          {arriveTangent=-0.2745, interpMode=RCIM_Cubic, leaveTangent=-0.2745, tangentMode=RCTM_User, time=1, value=0.1}
         ]}}}
      CurveIndex=$Particles.NormalizedChainPosition
      Scale Curve=1.5
     )
    )) {isSetParameters=true}
   &DynamicMaterialParameters(
    Index 0 Param 1=10
    Module.Index0Float0Write=true
    Module.Index0Float1Write=false
    Module.Index0Float2Write=false
    Module.Index0Float3Write=false
   )
  ]
  rendererClasses [&NiagaraRibbonRendererProperties]
  renderers [
   {rendererClass=&NiagaraRibbonRendererProperties}
   {rendererClass=&NiagaraRibbonRendererProperties, rendererIndex=1}
   {rendererClass=&NiagaraRibbonRendererProperties, rendererIndex=2}
  ]
 }
}
summary {
 systemName=NS_Character_BoldDormRise_Idle
 emitters [
  {emitterName=Plume_Flame, rendererClasses [&NiagaraRibbonRendererProperties]}
  {emitterName=CoreFlames, rendererClasses [&NiagaraMeshRendererProperties]}
  {bEnabled=false, emitterName=BasicGlow, rendererClasses [&NiagaraSpriteRendererProperties]}
  {emitterName=BasicSparks, rendererClasses [&NiagaraSpriteRendererProperties]}
  {emitterName=Flame_Licks, rendererClasses [&NiagaraMeshRendererProperties]}
  {
   emitterName=HelmetGlow
   rendererClasses [
    &NiagaraSpriteRendererProperties
    &NiagaraLightRendererProperties
    &NiagaraMeshRendererProperties
   ]
  }
  {emitterName=Flames, rendererClasses [&NiagaraSpriteRendererProperties]}
 ]
 userVariables [
  User.bIsADS {type=NiagaraBool}
  User.bIsFrontEnd {type=NiagaraBool}
  User.bIsFrontEndPreview {type=NiagaraBool}
  User.bisHeadHidden {type=NiagaraBool}
  User.Reactivity {type=NiagaraFloat, defaultValue=1, description=""}
  User.Variant {type=NiagaraFloat, defaultValue=0, description=""}
 ]
}
