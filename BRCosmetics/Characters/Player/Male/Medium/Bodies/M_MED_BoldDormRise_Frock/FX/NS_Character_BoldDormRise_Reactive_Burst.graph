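# Niagara burst effect for the BoldDormRise character cosmetic, firing a short flare of sparks, flame-lick meshes and a glow sprite at the helmet.
# All three active emitters attach to one skeletal mesh bone or socket and spawn only when System.showHeadFX is true.
# User.Variant blends colors from warm orange to white-grey, and extra glow attributes accompany the sparks and helmet glow.
# Three other emitters are present but disabled.
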NiagaraSystem /BRCosmetics/Characters/Player/Male/Medium/Bodies/M_MED_BoldDormRise_Frock/FX/NS_Character_BoldDormRise_Reactive_Burst
dependencies {
 usedDataInterfaces [
  &NiagaraDataInterfaceSpriteRendererInfo
  &NiagaraDataInterfaceCurve
  &NiagaraDataInterfaceColorCurve
  &NiagaraDataInterfaceSkeletalMesh
 ]
 usedDynamicInputs [
  &ClampFloat
  &Boolean_NotOperation
  &Boolean_AndOperation
  &SetBoolByFloatComparison
 ]
 usedModules [
  &self:SystemSpawnScript.SystemScriptSource.SystemScriptGraph.NiagaraNodeAssignment_0.SetVariables_#1
  &SystemState
  &FN_VariantManager
  &self:SystemSpawnScript.SystemScriptSource.SystemScriptGraph.NiagaraNodeAssignment_1.SetVariables_#2
  &self:SystemSpawnScript.SystemScriptSource.SystemScriptGraph.NiagaraNodeAssignment_3.SetVariables_#3
  &self:SystemSpawnScript.SystemScriptSource.SystemScriptGraph.NiagaraNodeAssignment_2.SetVariables_#4
 ]
}
emitters {
 BasicSparks {
  emitterUpdateScript [
   &EmitterState(Life Cycle Mode=::NewEnumerator1, Loop Behavior=::NewEnumerator1, Loop Duration=2, UseLoopDelay=false)
   &SpawnBurst_Instantaneous(
    Spawn Count=&MakeCustomIntFromBool(Bool=$System.showHeadFX, True INT=&RandomRangeInt(Minimum=20, Maximum=30), False INT=0)
    Use Spawn Probability=false
    UseLoopCountLimit=false
   )
   &SpawnRate(SpawnRate=10, Use Spawn Probability=false) {enabled=false}
  ]
  particleSpawnScript [
   &InitializeParticle(
    Color=&LerpLinearColors(
     StartColor=&RandomRangeLinearColor(Minimum=<rgba 1 0.475 0.1 1>, Maximum=<rgba 1 0.5725 0.05 1>)
     EndColor=&RandomRangeLinearColor(Minimum=<rgba 1 1 1 1>, Maximum=<rgba 0.7 0.7 0.7 1>)
     LerpFactor=$User.Variant
    )
    Lifetime=&RandomRangeFloat(Minimum=0.4, Maximum=1.1)
    Sprite Size=&Multiply_Vector2D_ByFloat(
     Float=&RandomRangeFloat(Minimum=0.5, Maximum=0.85)
     Vector2D=&RandomRangeVector2D(Minimum=<1 3>, Maximum=<1 5>)
    )
    Sprite Size Mode=::NewEnumerator4
    UsePositionOffset=false
    Write Material Random=true
   )
   &SkeletalMeshLocation(
    Bone / Socket Index=2
    Bone Sampling Mode=::NewEnumerator4
    MaskSpawnGroup=false
    Sampled Position Offset=<17 -5 0>
    Skeletal Mesh=$System.Skeletal Mesh
   )
   &SubUVAnimation(
    "Sprite Renderer "=DataInterface() {propertyValues {SpriteRenderer=&self:NiagaraEmitter_1.NiagaraSpriteRendererProperties_0}}
    SubUV Animation Mode:ENiagara_SubUVLookupModeV2=::NewEnumerator2
    UseEndFrame=false
    UseStartFrame=false
   ) {enabled=false}
   &ShapeLocation(MaskSpawnGroup=false, Override Seed=false, Sphere Radius=7, Sphere Surface Distribution=0.4524, UseOwnerScale=true)
   &AddVelocity(
    Module.Use Falloff Distance=false
    Origin Offset=<0 10 14>
    Override Seed=false
    Velocity Mode=::NewEnumerator1
    Velocity Origin=$Particles.SkeletalMeshLocation.SampledPosition
    Velocity Speed=&RandomRangeFloat(Minimum=40, Maximum=150)
   )
   &InheritVelocity(
    Inherited Velocity Amount Scale=&VectorFromFloat(Value=0.4)
    Inherited Velocity Speed Limit=10
    Speed Limit=true
    UseSpeedThreshold=false
   )
  ]
  particleUpdateScript [
   &ParticleState()
   &ScaleSpriteSize(
    Scale Factor=&MakeVector2D(
     X=1
     Y=&FloatFromCurve(FloatCurve=DataInterface() {propertyValues {Curve {keys [<at 0 1>,<at 0.7 0.5>]}}})
    )
    Scale Sprite Size Mode=::NewEnumerator2
   ) {module=ScaleSpriteSize001}
   &ScaleSpriteSize(Uniform Curve Sprite Scale=DataInterface() {propertyValues {Curve {keys [{tangentMode=RCTM_User, time=0, value=0}, <at 0.2 1>, {tangentMode=RCTM_User, time=1, value=0}]}}})
   &InheritSourceMovement(
    Applied Position Scale=&VectorFromFloat(Value=&FloatFromCurve(
      FloatCurve=DataInterface() {propertyValues {Curve {keys [{interpMode=RCIM_Cubic, tangentMode=RCTM_User, time=0.3, value=1}, {interpMode=RCIM_Cubic, tangentMode=RCTM_User, time=0.9, value=0}]}}}
      Scale Curve=&RandomRangeFloat(Minimum=0.7, Maximum=0.9)
     ))
    ApplyForce=false
    ApplyPosition=true
    ApplyPositionFalloff=false
   )
   &CurlNoiseForce(
    Bias Noise=false
    Noise Frequency=&RandomRangeFloat(Minimum=70, Maximum=110)
    Noise Strength=&RandomRangeFloat(Minimum=100, Maximum=200)
    Pan Noise=false
   )
   &Drag(Drag=&RandomRangeFloat(Minimum=2, Maximum=5), Use Linear Drag=true, Use Rotational Drag=true)
   &SolveForcesAndVelocity(Clamp Velocity=false, Limit Acceleration=false)
   &DynamicMaterialParameters(
    Index 0 Param 1=0.1
    Index 0 Param 2=4
    Module.Index0Float0Write=true
    Module.Index0Float1Write=true
    Module.Index0Float2Write=false
    Module.Index0Float3Write=false
   )
   &ScaleColor(
    Scale RGB=&MakeVectorFromLinearColorRGB(Color=&LerpLinearColors(
      StartColor=&ColorFromCurve(ColorCurve=DataInterface() {propertyValues {
         AlphaCurve {keys [<at 0 1 RCIM_Cubic>]}
         BlueCurve {keys [<at 0.0943 7 RCIM_Cubic>,<at 1 0.25>]}
         GreenCurve {keys [<at 0.0943 8.5 RCIM_Cubic>,<at 1 0.3583>]}
         RedCurve {keys [<at 0.0943 10 RCIM_Cubic>,<at 1 0.5>]}
        }})
      EndColor=&ColorFromCurve(ColorCurve=DataInterface() {propertyValues {
         AlphaCurve {keys [<at 0 1 RCIM_Cubic>]}
         BlueCurve {keys [<at 0.0943 10 RCIM_Cubic>,<at 1 0.2>]}
         GreenCurve {keys [<at 0.0943 10 RCIM_Cubic>,<at 1 0.2>]}
         RedCurve {keys [<at 0.0943 10 RCIM_Cubic>,<at 1 0.2>]}
        }})
      LerpFactor=$User.Variant
     ))
    ScaleA=false
    ScaleRGB=true
   )
   /SetVariables(
    Particles.ColorGlow:LinearColor=&MultiplyLinearColor(
     Linear Color=&LerpLinearColors(StartColor=<rgba 0.15 0.125 0.075 0.1>, EndColor=<rgba 0.1 0.1 0.1 0.1>, LerpFactor=$User.Variant)
     Scale Factor=&MakeVector4_FromLinearColor(LinearColor=$Particles.Color)
    )
    Particles.DynParGlow:Vector4f=<1 1 1 0>
    Particles.SpriteSizeGlow:Vector2f=&Multiply_Vector2D(A=$Particles.SpriteSize, B=<8 4>)
   ) {isSetParameters=true}
  ]
  rendererClasses [&NiagaraSpriteRendererProperties]
  renderers [
   {rendererClass=&NiagaraSpriteRendererProperties}
   {rendererClass=&NiagaraSpriteRendererProperties, rendererIndex=1}
  ]
 }
 Flame_Licks {
  emitterUpdateScript [
   &EmitterState(Life Cycle Mode=::NewEnumerator1, Loop Behavior=::NewEnumerator1, Loop Delay=0.04, Loop Duration=1, UseLoopDelay=true)
   &SpawnBurst_Instantaneous(
    Spawn Count=&MakeCustomIntFromBool(Bool=$System.showHeadFX, True INT=4, False INT=0)
    Use Spawn Probability=false
    UseLoopCountLimit=true
   )
  ]
  particleSpawnScript [
   &PartitionParticles(Number Of Partitions=4)
   &InitializeParticle(
    Color=<rgba 1.5 1.5 1.5 1>
    Lifetime Max=0.5
    Lifetime Min=0.35
    Lifetime Mode=::NewEnumerator1
    Mesh Scale=&Multiply_VectorByFloat(
     Vector=<1 1 1>
     Float=&MergeHelper_FloatSwitch(0=0.19, 1=0.15, 2=0.19, 3=0.15, Index=$Particles.PartitionParticles.Partition, Num=::NewEnumerator2)
    )
    Mesh Scale Mode=::NewEnumerator4
    UsePositionOffset=false
    Write Material Random=true
   )
   &InitialMeshOrientation(Mesh Axis To Orient=<0 1 0>, Mesh Orientation Mode=::NewEnumerator2, Rotation=<0.5 0 0>, Use Rotation Vector=true)
   &CameraOffset(Camera Offset Amount=5)
   &InitialMeshOrientation(
    Mesh Orientation Mode=::NewEnumerator2
    Rotation=&MergeHelper_VectorSwitch(
     0=<0.03 0 -0.04>
     1=<0.03 0 -0.09>
     2=<-0.03 0 -0.04>
     3=<-0.03 0 -0.09>
     Index=$Particles.PartitionParticles.Partition
     Num=::NewEnumerator2
    )
    Use Rotation Vector=true
   ) {module=InitialMeshOrientation001}
  ]
  particleUpdateScript [
   &SkeletalMeshLocation(
    Bone / Socket Index=2
    Bone Sampling Mode=::NewEnumerator4
    MaskSpawnGroup=false
    Sampled Position Offset=&AddVectorToPosition(
     Position=<15 0 0>
     Vector=&MergeHelper_VectorSwitch(0=<0 0 -1>, 1=<-3 -5 -3>, 2=<0 0 1>, 3=<-3 -5 3>, Index=$Particles.PartitionParticles.Partition, Num=::NewEnumerator2)
    )
    Skeletal Mesh=$System.Skeletal Mesh
   )
   &ParticleState()
   &ScaleMeshSize(Scale Factor=&VectorFromFloat(Value=&FloatFromCurve(FloatCurve=DataInterface() {propertyValues {Curve {keys [
          {arriveTangent=0.6848, interpMode=RCIM_Cubic, leaveTangent=0.6848, tangentMode=RCTM_User, time=0, value=0.5}
          {interpMode=RCIM_Cubic, tangentMode=RCTM_User, time=1, value=1.4}
         ]}}})))
   &DynamicMaterialParameters(
    Index 0 Param 2=&RandomRangeFloat(Minimum=0, Maximum=1)
    Index 0 Param 4=&Lerp_Float(
     A=0.15
     B=0.3
     Alpha=&FloatFromCurve(FloatCurve=DataInterface() {propertyValues {Curve {keys [
          {arriveTangent=5.3344, interpMode=RCIM_Cubic, leaveTangent=5.3344, tangentMode=RCTM_User, time=0, value=0}
          {arriveTangent=0.3, interpMode=RCIM_Cubic, leaveTangent=0.3, time=0.5, value=1}
          {arriveTangent=-1.3021, interpMode=RCIM_Cubic, leaveTangent=-1.3021, tangentMode=RCTM_User, time=1, value=0.3}
         ]}}})
    )
    Module.Index0Float0Write=true
    Module.Index0Float1Write=true
    Module.Index0Float2Write=false
    Module.Index0Float3Write=true
   )
   /SetVariables(Particles.ColorGlow:LinearColor=&MultiplyLinearColor(
     Linear Color=<rgba 1 0.4987 0.06 0.25>
     Scale Factor=&MakeVector4FromVectorAndFloat(
      "Vector (RGB)"=<1 1 1>
      "Float (Alpha)"=&FloatFromCurve(FloatCurve=DataInterface() {propertyValues {Curve {keys [{interpMode=RCIM_Cubic, tangentMode=RCTM_User, time=0, value=1}, {interpMode=RCIM_Cubic, tangentMode=RCTM_User, time=1, value=0}]}}})
     )
    )) {enabled=false, isSetParameters=true}
   /SetVariables(Particles.DynParGlow:Vector4f=&Multiply_Vector4(A=$Particles.DynamicMaterialParameter, B=<1 1 1 1.2>)) {enabled=false, isSetParameters=true}
   &ScaleColor(
    Scale Alpha=&FloatFromCurve(FloatCurve=DataInterface() {propertyValues {Curve {keys [{interpMode=RCIM_Cubic, tangentMode=RCTM_User, time=0.4, value=1}, {interpMode=RCIM_Cubic, tangentMode=RCTM_User, time=1, value=0}]}}})
    Scale RGB=&MakeVectorFromLinearColorRGB(Color=&LerpLinearColors(
      StartColor=&ColorFromCurve(ColorCurve=DataInterface() {propertyValues {
         AlphaCurve {keys [<at 0 1 RCIM_Cubic>]}
         BlueCurve {keys [<at 0.0449 19.8 RCIM_Cubic>,<at 0.3858 0.8>,<at 0.9447 0.132 RCIM_Cubic>]}
         GreenCurve {keys [<at 0.0449 21.6333 RCIM_Cubic>,<at 0.3858 0.9333>,<at 0.9447 0.264 RCIM_Cubic>]}
         RedCurve {keys [<at 0.0449 22 RCIM_Cubic>,<at 0.3858 1>,<at 0.9447 0.66 RCIM_Cubic>]}
        }})
      EndColor=&ColorFromCurve(ColorCurve=DataInterface() {propertyValues {
         AlphaCurve {keys [<at 0 1 RCIM_Cubic>]}
         BlueCurve {keys [<at 0.0628 10>,<at 0.3628 1>,<at 0.9063 0.7 RCIM_Cubic>]}
         GreenCurve {keys [<at 0.0628 10>,<at 0.3628 1>,<at 0.9063 0.7 RCIM_Cubic>]}
         RedCurve {keys [<at 0.0628 10>,<at 0.3628 1>,<at 0.9063 0.7 RCIM_Cubic>]}
        }})
      LerpFactor=$User.Variant
     ))
    ScaleA=true
    ScaleRGB=true
   )
  ]
  rendererClasses [&NiagaraMeshRendererProperties]
  renderers [
   {rendererClass=&NiagaraMeshRendererProperties, rendererIndex=0}
   {rendererClass=&NiagaraMeshRendererProperties, rendererIndex=1}
  ]
 }
 HelmetFakeLight {
  bEnabled=false
  emitterUpdateScript [
   &EmitterState(Life Cycle Mode=::NewEnumerator1, Loop Behavior=::NewEnumerator1, Loop Duration Mode=::NewEnumerator1, UseLoopDelay=false)
   &SpawnBurst_Instantaneous(Use Spawn Probability=false, UseLoopCountLimit=false)
  ]
  particleSpawnScript [
   &InitializeParticle(Color=<rgba 15 10 0 1>, Sprite Size Mode=::NewEnumerator3, Uniform Sprite Size=35, UsePositionOffset=false, Write Material Random=true)
   &CameraOffset(Camera Offset Amount=2)
  ]
  particleUpdateScript [
   &SkeletalMeshLocation(
    Bone / Socket Index=2
    Bone Sampling Mode=::NewEnumerator4
    MaskSpawnGroup=false
    Sampled Position Offset=<13 -4 0>
    Skeletal Mesh=$System.Skeletal Mesh
   )
   &ParticleState(Kill Particles When Lifetime Has Elapsed=false, Let Infinitely Lived Particles Die When Emitter Deactivates=true)
   &ScaleColor(ScaleA=true, ScaleRGB=true) {enabled=false}
  ]
  rendererClasses [&NiagaraSpriteRendererProperties]
  renderers [{rendererClass=&NiagaraSpriteRendererProperties}]
 }
 HelmetGlowLight {
  emitterUpdateScript [
   &EmitterState(Life Cycle Mode=::NewEnumerator1, Loop Behavior=::NewEnumerator1, Loop Duration Mode=::NewEnumerator1, UseLoopDelay=false)
   &SpawnBurst_Instantaneous(
    Spawn Count=&MakeCustomIntFromBool(Bool=$System.showHeadFX, True INT=1, False INT=0)
    Use Spawn Probability=false
    UseLoopCountLimit=true
   )
  ]
  particleSpawnScript [
   &InitializeParticle(
    Color=&LerpLinearColors(StartColor=<rgba 1 0.4633 0.08 1>, EndColor=<rgba 1 1 1 0.5>, LerpFactor=$User.Variant)
    Lifetime=0.4
    Sprite Rotation Mode=::NewEnumerator2
    Sprite Size Mode=::NewEnumerator3
    Uniform Sprite Size=80
    UsePositionOffset=false
    Write Material Random=true
   )
   &CameraOffset(Camera Offset Amount=10)
  ]
  particleUpdateScript [
   &SkeletalMeshLocation(
    Bone / Socket Index=2
    Bone Sampling Mode=::NewEnumerator4
    MaskSpawnGroup=false
    Sampled Position Offset=<13 -4 0>
    Skeletal Mesh=$System.Skeletal Mesh
   )
   &ScaleColor(
    Scale Alpha=&FloatFromCurve(FloatCurve=DataInterface() {propertyValues {Curve {keys [<at 0 1>,<at 1 0>]}}})
    Scale RGB=&MakeVectorFromLinearColorRGB(Color=&LerpLinearColors(
      StartColor=&ColorFromCurve(ColorCurve=DataInterface() {propertyValues {
         AlphaCurve {keys [<at 0 1 RCIM_Cubic>]}
         BlueCurve {keys [<at 0 1.4 RCIM_Cubic>,<at 0.6369 0.8 RCIM_Cubic>]}
         GreenCurve {keys [<at 0 1.7 RCIM_Cubic>,<at 0.6369 0.8 RCIM_Cubic>]}
         RedCurve {keys [<at 0 2 RCIM_Cubic>,<at 0.6369 1 RCIM_Cubic>]}
        }})
      EndColor=&ColorFromCurve(ColorCurve=DataInterface() {propertyValues {
         AlphaCurve {keys [<at 0 1 RCIM_Cubic>]}
         BlueCurve {keys [<at 0 1 RCIM_Cubic>,<at 0.8049 0.1 RCIM_Cubic>]}
         GreenCurve {keys [<at 0 1 RCIM_Cubic>,<at 0.8049 0.1 RCIM_Cubic>]}
         RedCurve {keys [<at 0 1 RCIM_Cubic>,<at 0.8049 0.1 RCIM_Cubic>]}
        }})
      LerpFactor=$User.Variant
     ))
    ScaleA=true
    ScaleRGB=true
   )
   /SetVariables(
    Particles.ColorGlow:LinearColor=&LerpLinearColors(
     StartColor=&ColorFromCurve(ColorCurve=DataInterface() {propertyValues {
        AlphaCurve {keys [<at 0 1 RCIM_Cubic>,<at 0.9747 0>]}
        BlueCurve {keys [<at 0.061 0.13 RCIM_Cubic>,<at 0.8838 0.04 RCIM_Cubic>]}
        GreenCurve {keys [<at 0.061 0.4925 RCIM_Cubic>,<at 0.8838 0.152 RCIM_Cubic>]}
        RedCurve {keys [<at 0.061 1 RCIM_Cubic>,<at 0.8838 1 RCIM_Cubic>]}
       }})
     EndColor=&ColorFromCurve(ColorCurve=DataInterface() {propertyValues {
        AlphaCurve {keys [<at 0 1 RCIM_Cubic>,<at 0.9747 0>]}
        BlueCurve {keys [<at 0 0.4 RCIM_Cubic>,<at 0.8814 0.1 RCIM_Cubic>]}
        GreenCurve {keys [<at 0 0.4 RCIM_Cubic>,<at 0.8814 0.1 RCIM_Cubic>]}
        RedCurve {keys [<at 0 0.4 RCIM_Cubic>,<at 0.8814 0.1 RCIM_Cubic>]}
       }})
     LerpFactor=$User.Variant
    )
    Particles.PositionGlow:NiagaraPosition=&AddVectorToPosition(
     Position=$Particles.Position
     Vector=&RotateVectorByQuaternion(Vector To Rotate=<5 -2 0>, Quaternion=$Particles.SkeletalMeshLocation.SampledRotation)
    )
   ) {isSetParameters=true}
   &ParticleState()
   &ScaleSpriteSize(Scale Sprite Size Mode=::NewEnumerator0, Uniform Scale Factor=&Lerp_Float(A=1, B=1.25, Alpha=$System.PlumeLength))
  ]
  rendererClasses [&NiagaraSpriteRendererProperties]
  renderers [
   {rendererClass=&NiagaraSpriteRendererProperties}
   {rendererClass=&NiagaraSpriteRendererProperties, rendererIndex=1}
  ]
 }
 Minimal002 {
  bEnabled=false
  emitterUpdateScript [
   &EmitterState(Life Cycle Mode=::NewEnumerator1, Loop Behavior=::NewEnumerator1, Loop Delay=0.05, Loop Duration=1, UseLoopDelay=true)
   &SpawnBurst_Instantaneous(Use Spawn Probability=false, UseLoopCountLimit=false)
  ]
  particleSpawnScript [
   &InitializeParticle(
    Color=<rgba 3 1.02 0.03 1>
    Lifetime=0.4
    Sprite Size Mode=::NewEnumerator3
    Uniform Sprite Size=100
    UsePositionOffset=false
    Write Material Random=true
   )
   &CameraOffset(Camera Offset Amount=30)
  ]
  particleUpdateScript [
   &SkeletalMeshLocation(
    Bone / Socket Index=2
    Bone Sampling Mode=::NewEnumerator4
    MaskSpawnGroup=false
    Orientation Sampling=::NewEnumerator1
    Sampled Position Offset=&MakeVector(
     X=&Lerp_Float(
      A=-6
      B=20
      Alpha=&FloatFromCurve(FloatCurve=DataInterface() {propertyValues {Curve {keys [
           {interpMode=RCIM_Cubic, tangentMode=RCTM_User, time=0, value=0.9}
           {interpMode=RCIM_Cubic, tangentMode=RCTM_User, time=0.2, value=1}
           {arriveTangent=-1.4888, interpMode=RCIM_Cubic, leaveTangent=-1.4888, tangentMode=RCTM_User, time=1, value=0}
          ]}}})
     )
     Y=&Lerp_Float(
      A=10
      B=-30
      Alpha=&FloatFromCurve(FloatCurve=DataInterface() {propertyValues {Curve {keys [
           {arriveTangent=1.6393, interpMode=RCIM_Cubic, leaveTangent=1.6393, tangentMode=RCTM_User, time=0, value=0}
           {arriveTangent=0.4053, interpMode=RCIM_Cubic, leaveTangent=0.4053, tangentMode=RCTM_User, time=0.8, value=1}
          ]}}})
     )
     Z=0
    )
    Skeletal Mesh=$System.Skeletal Mesh
   )
   /SetVariables(Particles.SpriteSizeLong:Vector2f=&Multiply_Vector2D_ByFloat(
     Float=&FloatFromCurve(FloatCurve=DataInterface() {propertyValues {Curve {keys [<at 0.1 0>,<at 0.4 1>,<at 1 0.5>]}}}, Scale Curve=50)
     Vector2D=<1 3>
    )) {isSetParameters=true}
   &InterpolateOverTime(Rate Of Change=20, Target Position Value:NiagaraPosition=$Particles.Position, Variable Type=::NewEnumerator6)
   &ParticleState()
   &ScaleSpriteSize(Uniform Curve Sprite Scale=DataInterface() {propertyValues {Curve {keys [{tangentMode=RCTM_User, time=0, value=1}, {tangentMode=RCTM_User, time=1, value=0.66}]}}})
   &SystemLocation(MaskSpawnGroup=false) {enabled=false}
   &ScaleColor(
    Scale Alpha=&FloatFromCurve(FloatCurve=DataInterface() {propertyValues {Curve {keys [<at 0.2 0>,<at 0.4 1>,<at 1 0>]}}}, Scale Curve=3)
    ScaleA=true
    ScaleRGB=true
   )
   &GravityForce(Gravity=<0 0 -1>)
   &SolveForcesAndVelocity(Clamp Velocity=false, Limit Acceleration=false)
   /Game/Sprite/FN_SpriteAlignmentAndFacing(
    SpriteAlignment:Vector3f=&NormalizeVector(
     Vector To Normalize=&SubtractVectorFromPosition(
      Position=$Particles.Position
      Vector=Linked() {linkedVariable {name=Particles.InterpolateOverTime.InterpolateOverTime_Position.Moving Average, type=NiagaraPosition}}
     )
     Safe Normalize=true
     Safe Fallback Threshold=0
     Safe Fallback Vector=<0 0 0.001>
    )
    SpriteFacing:Vector3f=<0 0 1>
   )
  ]
  rendererClasses [&NiagaraSpriteRendererProperties]
  renderers [
   {rendererClass=&NiagaraSpriteRendererProperties}
   {rendererClass=&NiagaraSpriteRendererProperties, rendererIndex=1}
  ]
 }
 Minimal005 {
  bEnabled=false
  emitterUpdateScript [
   &EmitterState(Life Cycle Mode=::NewEnumerator1, Loop Behavior=::NewEnumerator1, Loop Duration Mode=::NewEnumerator1, UseLoopDelay=false)
   &SpawnBurst_Instantaneous(Use Spawn Probability=false, UseLoopCountLimit=false)
  ]
  particleSpawnScript [
   &InitializeParticle(
    Color=<rgba 1 0.7163 0.149 0.4>
    Mesh Scale=&Multiply_VectorByFloat(Vector=<50 1 1>, Float=0.06)
    Mesh Scale Mode=::NewEnumerator4
    UsePositionOffset=false
    Write Material Random=true
   )
   &InitialMeshOrientation(Mesh Axis To Orient=<0 -1 0>, Mesh Orientation Mode=::NewEnumerator2, Rotation=<0 0 0.09>, Use Rotation Vector=true)
  ]
  particleUpdateScript [
   &SkeletalMeshLocation(
    Bone / Socket Index=2
    Bone Sampling Mode=::NewEnumerator4
    MaskSpawnGroup=false
    Sampled Position Offset=<17 -4 0>
    Skeletal Mesh=$System.Skeletal Mesh
   )
   &ParticleState(Kill Particles When Lifetime Has Elapsed=false, Let Infinitely Lived Particles Die When Emitter Deactivates=true)
   &ScaleColor(Scale Alpha=2, ScaleA=true, ScaleRGB=true)
   &DynamicMaterialParameters(Module.Index0Float0Write=true, Module.Index0Float1Write=false, Module.Index0Float2Write=false, Module.Index0Float3Write=false) {enabled=false}
  ]
  rendererClasses [&NiagaraMeshRendererProperties]
  renderers [{rendererClass=&NiagaraMeshRendererProperties, rendererIndex=0}]
 }
}
summary {
 systemName=NS_Character_BoldDormRise_Reactive_Burst
 emitters [
  {bEnabled=false, emitterName=Minimal005, rendererClasses [&NiagaraMeshRendererProperties]}
  {bEnabled=false, emitterName=HelmetFakeLight, rendererClasses [&NiagaraSpriteRendererProperties]}
  {emitterName=BasicSparks, rendererClasses [&NiagaraSpriteRendererProperties]}
  {emitterName=Flame_Licks, rendererClasses [&NiagaraMeshRendererProperties]}
  {emitterName=HelmetGlowLight, rendererClasses [&NiagaraSpriteRendererProperties]}
  {bEnabled=false, emitterName=Minimal002, rendererClasses [&NiagaraSpriteRendererProperties]}
 ]
 userVariables [
  User.bIsFrontEnd {type=NiagaraBool}
  User.bIsFrontEndPreview {type=NiagaraBool}
  User.bisHeadHidden {type=NiagaraBool}
  User.Reactivity {type=NiagaraFloat, defaultValue=1, description=""}
  User.Variant {type=NiagaraFloat, defaultValue=0, description=""}
 ]
}
