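# Physics asset for the M_MED_BoldDormRise_Frock male body cosmetic, defining bone-driven physics for the frock's cloth flaps.
# Pelvis, thigh and calf capsules cover the lower body, while a large offset root sphere tops out just above the root, likely approximating the ground.
# Limited-swing capsule chains hang from pelvis-linked base bones: a twist-locked back chain and two front chains.
# An angularly free constraint links the two front tips.
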PhysicsAsset /BRCosmetics/Characters/Player/Male/Medium/Bodies/M_MED_BoldDormRise_Frock/Meshes/M_MED_BoldDormRise_Frock_Cloth_Physics
bodies {
 calf_l [{shapeType=Capsule, length=45, radius=8, shapeName=calf_l_capsule, center=<-23.191 0 -0.8371>, rotation=<rot 90 180 0>}]
 calf_r [{shapeType=Capsule, length=45, radius=8, shapeName=calf_r_capsule, center=<23.191 0 0.8496>, rotation=<rot 90 0 0>}]
 cloth_01_bk [{shapeType=Capsule, length=9, radius=4.5, shapeName=cloth_01_bk_capsule, center=<7.7671 0 -0.7446>, rotation=<rot 0 0 90>}]
 cloth_01_fr_l [{shapeType=Capsule, length=4, radius=6, shapeName=cloth_01_fr_l_capsule, center=<4.8585 1.1476 -2.2895>, rotation=<rot 90 0 0>}]
 cloth_01_fr_r [{shapeType=Capsule, length=5.5, radius=6, shapeName=cloth_01_fr_r_capsule, center=<4.8585 -1.1257 -2.0887>, rotation=<rot 90 0 0>}]
 cloth_02_bk [{shapeType=Capsule, length=10.5, radius=4.5, shapeName=cloth_02_bk_capsule, center=<3.4957 0 0>, rotation=<rot 0 0 90>}]
 cloth_02_fr_l [{shapeType=Capsule, length=7, radius=6, shapeName=cloth_02_fr_l_capsule, center=<6.7417 1.6895 -2.24>, rotation=<rot 90 0 0>}]
 cloth_02_fr_r [{shapeType=Capsule, length=10, radius=6, shapeName=cloth_02_fr_r_capsule, center=<6.7417 -1.1346 -1.521>, rotation=<rot 90 0 0>}]
 cloth_03_bk [{shapeType=Capsule, length=11, radius=4.5, shapeName=cloth_03_bk_capsule, center=<3.2451 0 0>, rotation=<rot 0 0 90>}]
 cloth_03_fr_l [{shapeType=Capsule, length=8, radius=4, shapeName=cloth_03_fr_l_capsule, center=<7.0216 -1.5268 -0.2302>, rotation=<rot 90 -30.865 0>}]
 cloth_03_fr_r [{shapeType=Capsule, length=8, radius=4, shapeName=cloth_03_fr_r_capsule, center=<7.4979 1.5093 0.033>, rotation=<rot 90 38.9726 0>}]
 cloth_04_bk [{shapeType=Capsule, length=1, radius=4.5, shapeName=cloth_04_bk_capsule, center=<6.8305 0 0>, rotation=<rot 0 0 90>}]
 cloth_bk_base [{shapeType=Capsule, length=0.505, radius=0.505, shapeName=cloth_bk_base_capsule, rotation=<rot 0 -90 0>}]
 cloth_fr_base_l [{shapeType=Capsule, length=0.505, radius=0.505, shapeName=cloth_fr_base_l_capsule}]
 cloth_fr_base_r [{shapeType=Capsule, length=0.505, radius=0.505, shapeName=cloth_fr_base_r_capsule}]
 pelvis [{shapeType=Capsule, length=9.5, radius=9, shapeName=pelvis_capsule, center=<0.0218 0 0>}]
 root [{shapeType=Sphere, radius=189.5, shapeName=root_sphere, center=<0 0 -186.7306>}]
 thigh_l [{shapeType=Capsule, length=22, radius=8, shapeName=thigh_l_capsule, center=<-17.5217 -1.0663 0>, rotation=<rot -90 0 0>}]
 thigh_r [{shapeType=Capsule, length=25, radius=8, shapeName=thigh_r_capsule, center=<17.5217 0.7728 0>, rotation=<rot 90 0 0>}]
}
constraints [
 {
  bone1Name=cloth_01_bk
  bone2Name=cloth_bk_base
  swing1LimitDegrees=10
  swing1Motion=Locked
  swing2LimitDegrees=20
  twistLimitDegrees=5
  twistMotion=Locked
 }
 {bone1Name=cloth_02_bk, bone2Name=cloth_01_bk, swing1LimitDegrees=5, swing2LimitDegrees=15, twistLimitDegrees=5, twistMotion=Locked}
 {bone1Name=cloth_03_bk, bone2Name=cloth_02_bk, swing1LimitDegrees=10, swing2LimitDegrees=20, twistLimitDegrees=5, twistMotion=Locked}
 {
  bone1Name=calf_l
  bone2Name=thigh_l
  swing1LimitDegrees=45
  swing1Motion=Free
  swing2LimitDegrees=45
  swing2Motion=Free
  twistLimitDegrees=45
  twistMotion=Free
 }
 {
  bone1Name=calf_r
  bone2Name=thigh_r
  swing1LimitDegrees=45
  swing1Motion=Free
  swing2LimitDegrees=45
  swing2Motion=Free
  twistLimitDegrees=45
  twistMotion=Free
 }
 {
  bone1Name=thigh_l
  bone2Name=pelvis
  swing1LimitDegrees=45
  swing1Motion=Free
  swing2LimitDegrees=45
  swing2Motion=Free
  twistLimitDegrees=45
  twistMotion=Free
 }
 {
  bone1Name=thigh_r
  bone2Name=pelvis
  swing1LimitDegrees=45
  swing1Motion=Free
  swing2LimitDegrees=45
  swing2Motion=Free
  twistLimitDegrees=45
  twistMotion=Free
 }
 {
  bone1Name=cloth_bk_base
  bone2Name=pelvis
  swing1LimitDegrees=45
  swing1Motion=Free
  swing2LimitDegrees=45
  swing2Motion=Free
  twistLimitDegrees=45
  twistMotion=Free
 }
 {
  bone1Name=cloth_fr_base_r
  bone2Name=pelvis
  swing1LimitDegrees=45
  swing1Motion=Free
  swing2LimitDegrees=45
  swing2Motion=Free
  twistLimitDegrees=45
  twistMotion=Free
 }
 {
  bone1Name=cloth_fr_base_l
  bone2Name=pelvis
  swing1LimitDegrees=45
  swing1Motion=Free
  swing2LimitDegrees=45
  swing2Motion=Free
  twistLimitDegrees=45
  twistMotion=Free
 }
 {bone1Name=cloth_01_fr_l, bone2Name=cloth_fr_base_l, swing1LimitDegrees=15, swing2LimitDegrees=40, twistLimitDegrees=15}
 {bone1Name=cloth_02_fr_l, bone2Name=cloth_01_fr_l, swing1LimitDegrees=20, swing2LimitDegrees=30, twistLimitDegrees=1}
 {bone1Name=cloth_03_fr_l, bone2Name=cloth_02_fr_l, swing1LimitDegrees=15, swing2LimitDegrees=30, twistLimitDegrees=1}
 {bone1Name=cloth_01_fr_r, bone2Name=cloth_fr_base_r, swing1LimitDegrees=15, swing2LimitDegrees=40, twistLimitDegrees=15}
 {bone1Name=cloth_02_fr_r, bone2Name=cloth_01_fr_r, swing1LimitDegrees=20, swing2LimitDegrees=30, twistLimitDegrees=1}
 {bone1Name=cloth_03_fr_r, bone2Name=cloth_02_fr_r, swing1LimitDegrees=15, swing2LimitDegrees=30, twistLimitDegrees=1}
 {
  bone1Name=cloth_03_fr_r
  bone2Name=cloth_03_fr_l
  swing1LimitDegrees=45
  swing1Motion=Free
  swing2LimitDegrees=45
  swing2Motion=Free
  twistLimitDegrees=45
  twistMotion=Free
 }
 {bone1Name=cloth_04_bk, bone2Name=cloth_03_bk, swing1LimitDegrees=10, swing2LimitDegrees=15, twistLimitDegrees=5, twistMotion=Locked}
 {
  bone1Name=pelvis
  bone2Name=root
  swing1LimitDegrees=45
  swing1Motion=Free
  swing2LimitDegrees=45
  swing2Motion=Free
  twistLimitDegrees=45
  twistMotion=Free
 }
]
