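# Post-copy animation Blueprint for the BoldDormRise Frock male player body cosmetic, adding rig-driven and simulated secondary motion.
# It copies the attached parent's pose into a Control Rig fed DBNO and RBAN flags, then runs RigidBody cloth physics in pelvis space.
# RBAN is the inverse of bRigidBodyEnabled.
# Each update also computes a state-based Gravity Override, but the RigidBody node uses fixed downward gravity instead.
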AnimBlueprint /BRCosmetics/Characters/Player/Male/Medium/Bodies/M_MED_BoldDormRise_Frock/Meshes/M_MED_BoldDormRise_Frock_AnimBP
target_skeleton=&/M_MED_BoldDormRise_Frock_Skeleton
graphs [
 AnimGraph {nodes [
   n1 = Control Rig!(Source=Copy Pose From Mesh() {bUseAttachedParent=true}, Alpha, DBNO:bool=Get bIsDBNO(), RBAN:bool=Get RBAN() -> (RBAN:bool)) -> (Pose) {controlRigAssetReference {blueprintRigClass=&/M_MED_BoldDormRise_Frock_CtrlRig.M_MED_BoldDormRise_Frock_CtrlRig_C}}
   n2 = RigidBody!(ComponentPose=Local To Component(LocalPose=@n1), Alpha, OverrideWorldGravity:Vector="0,0,-980", ExternalForce) -> (Pose) {
    bClampLinearTranslationLimitToRefPose=true
    bOverrideWorldGravity=true
    overridePhysicsAsset=&/M_MED_BoldDormRise_Frock_Cloth_Physics
    simulationSpace=BaseBoneSpace
    baseBoneRef {boneName=pelvis}
    componentAppliedLinearAccClamp=<3000 3000 3000>
    overrideWorldGravity {z=-980}
    simSpaceSettings {
     dampingAlpha=0
     maxAngularAcceleration=40
     maxAngularVelocity=8
     maxLinearAcceleration=3500
     maxLinearVelocity=700
     velocityScaleZ=0.6
     worldAlpha=1
     externalLinearDragV=<1 1 2>
    }
   }
   Output Pose(Result=Component To Local(ComponentPose=@n2)) {layerGroup=None}
  ]}
 EventGraph {nodes [
   Event BlueprintUpdateAnimation()
   > n1 = Sequence()
   > n2 = "Gravity Override State Function (Parameter Asset) (Nativized)"(
    CharacterPartAnimInstance=Self-Reference()
    GameTimeInSeconds=GetGameTimeInSeconds()
    Parameters=&/M_MED_BoldDormRise_Frock_Parameters
   )
   > Set Gravity Override(Gravity Override=@n2)
   Set RBAN(execute=@n1.then_1, RBAN:bool=NOT Boolean(A=Get bRigidBodyEnabled())) -> (Output_Get:bool)
  ]}
]
variables [
 Gravity Override {type=Vector}
 Gravity Override Params {
  type=GravityOverrideParamsStruct
  default {dBNO {jointName=root, gravityOverride=<0 0 -980>}, skydive_GoingUp {gravityOverride=<0 -300 -980>}}
 }
 RBAN {type=boolean}
]
