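# Physics asset of cape colliders for the medium male BoldDormRise Frock body cosmetic, giving body shapes for the cape to collide with.
# Each body is a single capsule, covering pelvis, spine, shoulder pads, arms, hands, legs and feet, with no head body.
# Four identical capsules on the back cloth bones form a chain constrained to the pelvis.
# That cloth chain and the lower spine joints explicitly have free swing and twist.
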PhysicsAsset /BRCosmetics/Characters/Player/Male/Medium/Bodies/M_MED_BoldDormRise_Frock/Meshes/M_MED_BoldDormRise_Frock_CapeColliders
bodies {
 calf_l [{shapeType=Capsule, length=39.525, radius=10.7547, center=<-24.2876 -0.588 0>, rotation=<rot 90 176.3495 0>}]
 calf_r [{shapeType=Capsule, length=39.525, radius=10.7547, center=<24.2876 0.588 0>, rotation=<rot 90 -3.6505 0>}]
 cloth_01_bk [{shapeType=Capsule, length=12, radius=7, shapeName=cloth_01_bk_capsule, center=<3.1292 0 0>, rotation=<rot 0 0 90>}]
 cloth_02_bk [{shapeType=Capsule, length=12, radius=7, shapeName=cloth_02_bk_capsule, center=<-0.1127 0 0>, rotation=<rot 0 0 90>}]
 cloth_03_bk [{shapeType=Capsule, length=12, radius=7, shapeName=cloth_03_bk_capsule, center=<1.0076 0 0>, rotation=<rot 0 0 90>}]
 cloth_04_bk [{shapeType=Capsule, length=12, radius=7, shapeName=cloth_04_bk_capsule, center=<7.036 0 0>, rotation=<rot 0 0 90>}]
 foot_l [{shapeType=Capsule, length=16.5855, radius=10.4926, center=<-3.9328 7.933 0>, rotation=<rot 0 4.5561 -90>}]
 foot_r [{shapeType=Capsule, length=16.5855, radius=10.4926, center=<3.9328 -7.933 0>, rotation=<rot 0 4.5561 90>}]
 hand_l [{shapeType=Capsule, length=7.3409, radius=8.104, center=<7.7246 0 0>, rotation=<rot 90 172.7826 0>}]
 hand_r [{shapeType=Capsule, length=7.3409, radius=8.104, center=<-7.7246 0 0>, rotation=<rot 90 -7.2174 0>}]
 lowerarm_l [{shapeType=Capsule, length=19.9928, radius=8.1273, center=<10.2847 0.5092 1.0797>, rotation=<rot 90 169.4523 0>}]
 lowerarm_r [{shapeType=Capsule, length=19.9928, radius=8.1273, center=<-10.2847 -0.5092 -1.0797>, rotation=<rot 90 -10.5477 0>}]
 pelvis [{shapeType=Capsule, length=9.6713, radius=14.4245}]
 spine_01 [{shapeType=Capsule, length=13.005, radius=12.005, shapeName=spine_01_capsule}]
 spine_02 [{shapeType=Capsule, length=8.9206, radius=12.7176, shapeName=spine_02_capsule}]
 spine_03 [{shapeType=Capsule, length=8.3562, radius=14.3642}]
 spine_04 [{shapeType=Capsule, length=6.6367, radius=16.5554, center=<0 1.7472 0>}]
 spine_05 [{shapeType=Capsule, length=12.4938, radius=17.1134}]
 support_shoulderpad_base_l [{shapeType=Capsule, length=12.1529, radius=9.2371, center=<0 0 3.4947>, rotation=<rot 90 180 0>}]
 support_shoulderpad_base_r [{shapeType=Capsule, length=12.1529, radius=9.2371, center=<0 0 -3.4947>, rotation=<rot 90 0 0>}]
 thigh_l [{shapeType=Capsule, length=28.4775, radius=10.6677, center=<-17.9655 -1.6135 0>, rotation=<rot 90 180 0>}]
 thigh_r [{shapeType=Capsule, length=28.4775, radius=10.6677, center=<17.9655 1.6135 0>, rotation=<rot 90 0 0>}]
 upperarm_l [{shapeType=Capsule, length=19.9654, radius=8.8599, center=<8.503 0 0>, rotation=<rot 90 180 0>}]
 upperarm_r [{shapeType=Capsule, length=19.9654, radius=8.8599, center=<-8.503 0 0>, rotation=<rot 90 0 0>}]
}
constraints [
 {bone1Name=spine_03, bone2Name=spine_02, swing1LimitDegrees=45, swing2LimitDegrees=45, twistLimitDegrees=45}
 {bone1Name=spine_04, bone2Name=spine_03, swing1LimitDegrees=45, swing2LimitDegrees=45, twistLimitDegrees=45}
 {bone1Name=spine_05, bone2Name=spine_04, swing1LimitDegrees=45, swing2LimitDegrees=45, twistLimitDegrees=45}
 {bone1Name=upperarm_l, bone2Name=spine_05, swing1LimitDegrees=45, swing2LimitDegrees=45, twistLimitDegrees=45}
 {bone1Name=lowerarm_l, bone2Name=upperarm_l, swing1LimitDegrees=45, swing2LimitDegrees=45, twistLimitDegrees=45}
 {bone1Name=hand_l, bone2Name=lowerarm_l, swing1LimitDegrees=45, swing2LimitDegrees=45, twistLimitDegrees=45}
 {bone1Name=support_shoulderpad_base_l, bone2Name=upperarm_l, swing1LimitDegrees=45, swing2LimitDegrees=45, twistLimitDegrees=45}
 {bone1Name=upperarm_r, bone2Name=spine_05, swing1LimitDegrees=45, swing2LimitDegrees=45, twistLimitDegrees=45}
 {bone1Name=lowerarm_r, bone2Name=upperarm_r, swing1LimitDegrees=45, swing2LimitDegrees=45, twistLimitDegrees=45}
 {bone1Name=hand_r, bone2Name=lowerarm_r, swing1LimitDegrees=45, swing2LimitDegrees=45, twistLimitDegrees=45}
 {bone1Name=support_shoulderpad_base_r, bone2Name=upperarm_r, swing1LimitDegrees=45, swing2LimitDegrees=45, twistLimitDegrees=45}
 {bone1Name=thigh_l, bone2Name=pelvis, swing1LimitDegrees=45, swing2LimitDegrees=45, twistLimitDegrees=45}
 {bone1Name=calf_l, bone2Name=thigh_l, swing1LimitDegrees=45, swing2LimitDegrees=45, twistLimitDegrees=45}
 {bone1Name=foot_l, bone2Name=calf_l, swing1LimitDegrees=45, swing2LimitDegrees=45, twistLimitDegrees=45}
 {bone1Name=thigh_r, bone2Name=pelvis, swing1LimitDegrees=45, swing2LimitDegrees=45, twistLimitDegrees=45}
 {bone1Name=calf_r, bone2Name=thigh_r, swing1LimitDegrees=45, swing2LimitDegrees=45, twistLimitDegrees=45}
 {bone1Name=foot_r, bone2Name=calf_r, swing1LimitDegrees=45, swing2LimitDegrees=45, twistLimitDegrees=45}
 {
  bone1Name=spine_01
  bone2Name=pelvis
  swing1LimitDegrees=45
  swing1Motion=Free
  swing2LimitDegrees=45
  swing2Motion=Free
  twistLimitDegrees=45
  twistMotion=Free
 }
 {
  bone1Name=spine_02
  bone2Name=spine_01
  swing1LimitDegrees=45
  swing1Motion=Free
  swing2LimitDegrees=45
  swing2Motion=Free
  twistLimitDegrees=45
  twistMotion=Free
 }
 {
  bone1Name=cloth_04_bk
  bone2Name=cloth_03_bk
  swing1LimitDegrees=45
  swing1Motion=Free
  swing2LimitDegrees=45
  swing2Motion=Free
  twistLimitDegrees=45
  twistMotion=Free
 }
 {
  bone1Name=cloth_02_bk
  bone2Name=cloth_01_bk
  swing1LimitDegrees=45
  swing1Motion=Free
  swing2LimitDegrees=45
  swing2Motion=Free
  twistLimitDegrees=45
  twistMotion=Free
 }
 {
  bone1Name=cloth_03_bk
  bone2Name=cloth_02_bk
  swing1LimitDegrees=45
  swing1Motion=Free
  swing2LimitDegrees=45
  swing2Motion=Free
  twistLimitDegrees=45
  twistMotion=Free
 }
 {
  bone1Name=cloth_01_bk
  bone2Name=pelvis
  swing1LimitDegrees=45
  swing1Motion=Free
  swing2LimitDegrees=45
  swing2Motion=Free
  twistLimitDegrees=45
  twistMotion=Free
 }
]
